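# Modular Control Rig for a biped spine, also usable as a neck, built from a Start-to-End connector bone chain.
# Construction spawns three primary FK controls, with Start, Mid and End IK controls under that hierarchy.
# It adds per-bone secondary FKs, match nulls, gimbals, space nulls and stretch, rotation and visibility channels.
# The IK drivers define a spline, and each bone stores its distance ratio along it as metadata.
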ControlRigBlueprint /EpicControlRig/Modules/Biped/Spine/CRM_Epic_Spine_v01
graphs [
 Concat_1_ContainedGraph {nodes [
   n1 = RigVMFunctionEntryNode()
   Concat = /Script/RigVM.RigVMFunction_NameConcat(A=@n1.A, B=@n1.B) {is=struct RigVMUnitNode, template="Concat::Execute(in A,in B,out Result)"}
   RigVMFunctionReturnNode(Result=&RigVMFunction_NameConcat(A=@Concat.Result, B=@n1.C) {template="Concat::Execute(in A,in B,out Result)"}.Result)
  ]}
 ConstructionGraph {nodes [
   n1 = &RigUnit_PrepareForExecution()
   n2 = /Script/ControlRig.RigUnit_HierarchyGetChainItemArray(End=(Type=Connector,Name="End"), Start=(Type=Connector,Name="Start"), bIncludeEnd=True, bIncludeStart=True, bReverse=False) {is=struct RigVMUnitNode}
   RigVMVariableNode(Value=@n2.Chain, ExecuteContext=@n1.ExecutePin) {variable=Bones}
   > RigVMVariableNode(Value=(Type=Connector,Name="EndParent")) {variable=End Parent}
   > RigVMVariableNode(Value=(Type=Connector,Name="StartParent")) {variable=Start Parent}
   n3 = /Script/RigVM.RigVMFunction_MathTransformArrayToSRT(Transforms=/Script/ControlRig.RigUnit_GetTransformItemArray(Space=GlobalSpace, bInitial=true, Items=RigVMVariableNode() {variable=Bones}.Value) {is=struct RigVMUnitNode}.Transforms) {is=struct RigVMUnitNode}
   > RigVMVariableNode(Value=RigVMDispatchNode(Index=0, Array=@n3.Translations) {template="DISPATCH_RigVMDispatch_ArrayGetAtIndex(in Array,in Index,out Element)"}.Element) {variable=Start Position}
   > RigVMVariableNode(Value=RigVMDispatchNode(Index=-1, Array=@n3.Translations) {template="DISPATCH_RigVMDispatch_ArrayGetAtIndex(in Array,in Index,out Element)"}.Element) {variable=End Position}
   Get Control Color From Metadata = RigVMFunctionReferenceNode(Color Override=RigVMVariableNode() {variable=Color}.Value)
   > RigVMVariableNode(Value=@"Get Control Color From Metadata".Result) {variable=Color}
   > RigVMVariableNode(Value=/Script/RigVM.RigVMFunction_MathColorMul(B=(R=0.5,G=0.2,B=0.2,A=1), A=RigVMVariableNode() {variable=Color}.Value) {is=struct RigVMUnitNode, template="Multiply::Execute(in A,in B,out Result)"}.Result) {variable=Alt Color}
   > RigVMFunction_Sequence = RigVMAggregateNode()
   Build Spline From Items_1 = RigVMFunctionReferenceNode(
    Start Control=RigVMVariableNode() {variable=Start Driver}.Value
    End Control=RigVMVariableNode() {variable=End Driver}.Value
    Mid Control=RigVMVariableNode() {variable=Mid Control}.Value
   )
   RigVMVariableNode(Value=@"Build Spline From Items_1".Spline, ExecuteContext=@RigVMFunction_Sequence.E) {variable=Spline}
   > RigVMVariableNode(Value=/Script/ControlRigSpline.RigUnit_GetLengthControlRigSpline(Spline=RigVMVariableNode() {variable=Spline}.Value) {is=struct RigVMUnitNode}.Length) {variable=Spline Orig Length}
   n4 = RigVMVariableNode() {variable=Bones}
   > n5 = RigVMDispatchNode(Array=@n4.Value) {template="DISPATCH_RigVMDispatch_ArrayIterator(in Array,out Element,out Index,out Count,out Ratio)"}
   DISPATCH_RigVMDispatch_ArrayGetAtIndex = RigVMDispatchNode(Index=0, Array=@n4.Value) {template="DISPATCH_RigVMDispatch_ArrayGetAtIndex(in Array,in Index,out Element)"}
   RigVMFunction_MathVectorDistance = &RigVMFunction_MathVectorDistance(
    B=&RigUnit_GetTransform(bInitial=true, Item=@n5.Element).Transform.Translation
    A=&RigUnit_GetTransform(bInitial=true, Item=@DISPATCH_RigVMDispatch_ArrayGetAtIndex.Element).Transform.Translation
   )
   Divide = &RigVMFunction_MathDoubleDiv(A=@RigVMFunction_MathVectorDistance.Result, B=RigVMVariableNode() {variable=Spline Orig Length}.Value)
   > RigVMDispatchNode(Name=Bone Percentage, NameSpace=Self, Item=@n5.Element, Value=@Divide.Result) {template="DISPATCH_RigDispatch_SetMetadata(in Item,in Name,in NameSpace,in Value,out Success)"}
   RigVMDispatchNode(Array=RigVMVariableNode() {variable=IK Gimbal Controls}.Value, ExecuteContext=@RigVMFunction_Sequence.B) {template="DISPATCH_RigVMDispatch_ArrayReset(io Array)"}
   n6 = RigVMVariableNode() {variable=Bones}
   > n7 = &RigUnit_HierarchyAddNull(
    Name=Default Start IK Space
    Space=GlobalSpace
    Transform=&RigUnit_GetTransform(
     bInitial=true
     Item=RigVMDispatchNode(Index=0, Array=@n6.Value) {template="DISPATCH_RigVMDispatch_ArrayGetAtIndex(in Array,in Index,out Element)"}.Element
    ).Transform
    Parent=RigVMVariableNode() {variable=Start Parent}.Value
   )
   n8 = RigVMVariableNode() {variable=IKs Control Shapes}
   Control Color Override From Metadata_1 = RigVMFunctionReferenceNode(Color Override=@n8.Value.Color)
   n9 = RigVMVariableNode() {variable=Rotation Offset}
   n10 = RigVMVariableNode() {variable=Control Scale Factor Profile}
   > n11 = /Script/ControlRig.RigUnit_HierarchyAddControlTransform(
    InitialValue=(Rotation=(X=0,Y=0,Z=0,W=1),Translation=(X=0,Y=0,Z=0),Scale3D=(X=1,Y=1,Z=1))
    Name=Start
    OffsetSpace=GlobalSpace
    OffsetTransform=(Rotation=(X=0,Y=0,Z=0,W=1),Translation=(X=0,Y=0,Z=0),Scale3D=(X=1,Y=1,Z=1))
    Settings="(InitialSpace=LocalSpace,bUsePreferredRotationOrder=true,PreferredRotationOrder=YZX,Limits=(LimitTranslationX=(bMinimum=False,bMaximum=False),LimitTranslationY=(bMinimum=False,bMaximum=False),LimitTranslationZ=(bMinimum=False,bMaximum=False),LimitPitch=(bM ...(truncated)"
    Settings.Shape.Color=@"Control Color Override From Metadata_1".Result
    Parent=@n7.Item
    Settings.Shape.Transform.Rotation=/Script/RigVM.RigVMFunction_MathQuaternionMul(B=@n9.Value, A=@n8.Value.Transform.Rotation) {is=struct RigVMUnitNode, template="Multiply::Execute(in A,in B,out Result)"}.Result
    Settings.Shape.Transform.Scale3D=/Script/RigVM.RigVMFunction_MathVectorScale(
     Value=@n8.Value.Transform.Scale3D
     Factor=&RigVMFunction_AnimEvalRichCurve(SourceMinimum=0, TargetMinimum=0, Value=0, Curve=@n10.Value).Result
    ) {is=struct RigVMUnitNode, template="Scale::Execute(in Value,in Factor,out Result)"}.Result
    Settings.DisplayName=RigVMVariableNode() {variable=IK Start Display Name}.Value
    OffsetTransform.Translation=RigVMVariableNode() {variable=Start Position}.Value
    OffsetTransform.Rotation=@n9.Value
    Settings.Shape.Name=@n8.Value.Name
    Settings.Shape.Transform.Translation=@n8.Value.Transform.Translation
    Settings.Shape.bVisible=@n8.Value.bVisible
    Settings.PreferredRotationOrder=RigVMVariableNode() {variable=Rotation Order}.Value
   ) {
    is=struct RigVMUnitNode
    template="SpawnControl::Execute(in InitialValue,in Settings,in OffsetTransform,in OffsetSpace,in Parent,in Name,out Item)"
   }
   n12 = RigVMVariableNode() {variable=Bones}
   > Set_Metadata = RigVMDispatchNode(
    Name=Offset
    NameSpace=Self
    Value=&RigUnit_GetRelativeTransformForItem(
     bChildInitial=true
     bParentInitial=true
     Parent=RigVMDispatchNode(Index=0, Array=@n12.Value) {template="DISPATCH_RigVMDispatch_ArrayGetAtIndex(in Array,in Index,out Element)"}.Element
     Child=@n11.Item
    ).RelativeTransform
    Item=@n11.Item
   ) {template="DISPATCH_RigDispatch_SetMetadata(in Item,in Name,in NameSpace,in Value,out Success)"}
   > RigVMVariableNode(Value=@n11.Item) {variable=Start Control}
   > RigVMVariableNode(Value=@n11.Item) {variable=Start Driver}
   > n13 = &RigUnit_HierarchyAddNull(
    Name=End
    Space=GlobalSpace
    Transform=(Rotation=(X=0,Y=0,Z=0,W=1),Translation=(X=0,Y=0,Z=0),Scale3D=(X=1,Y=1,Z=1))
    Transform.Translation=RigVMVariableNode() {variable=Start Position}.Value
    Parent=RigVMVariableNode() {variable=End Parent}.Value
   )
   > RigVMVariableNode(Value=@n13.Item) {variable=Chest Null}
   n14 = RigVMVariableNode() {variable=Bones}
   > n15 = &RigUnit_HierarchyAddNull(
    Name=Default End IK Space
    Space=GlobalSpace
    Transform=&RigUnit_GetTransform(
     bInitial=true
     Item=RigVMDispatchNode(Index=-1, Array=@n14.Value) {template="DISPATCH_RigVMDispatch_ArrayGetAtIndex(in Array,in Index,out Element)"}.Element
    ).Transform
    Parent=@n13.Item
   )
   n16 = RigVMVariableNode() {variable=IKs Control Shapes}
   Control Color Override From Metadata_2 = RigVMFunctionReferenceNode(Color Override=@n16.Value.Color)
   n17 = RigVMVariableNode() {variable=Rotation Offset}
   n18 = RigVMVariableNode() {variable=Control Scale Factor Profile}
   n19 = RigVMVariableNode() {variable=IK End Display Name}
   > n20 = &RigUnit_HierarchyAddControlTransform(
    InitialValue=(Rotation=(X=0,Y=0,Z=0,W=1),Translation=(X=0,Y=0,Z=0),Scale3D=(X=1,Y=1,Z=1))
    Name=End
    OffsetSpace=GlobalSpace
    OffsetTransform=(Rotation=(X=0,Y=0,Z=0,W=1),Translation=(X=0,Y=0,Z=0),Scale3D=(X=1,Y=1,Z=1))
    Settings="(InitialSpace=LocalSpace,bUsePreferredRotationOrder=true,PreferredRotationOrder=YZX,Limits=(LimitTranslationX=(bMinimum=False,bMaximum=False),LimitTranslationY=(bMinimum=False,bMaximum=False),LimitTranslationZ=(bMinimum=False,bMaximum=False),LimitPitch=(bM ...(truncated)"
    Settings.Shape.Color=@"Control Color Override From Metadata_2".Result
    Parent=@n15.Item
    Settings.Shape.Transform.Rotation=&RigVMFunction_MathQuaternionMul(B=@n17.Value, A=@n16.Value.Transform.Rotation) {template="Multiply::Execute(in A,in B,out Result)"}.Result
    Settings.Shape.Transform.Scale3D=&RigVMFunction_MathVectorScale(
     Value=@n16.Value.Transform.Scale3D
     Factor=&RigVMFunction_AnimEvalRichCurve(SourceMinimum=0, TargetMinimum=0, Value=1, Curve=@n18.Value).Result
    ) {template="Scale::Execute(in Value,in Factor,out Result)"}.Result
    OffsetTransform.Translation=RigVMVariableNode() {variable=End Position}.Value
    OffsetTransform.Rotation=@n17.Value
    Settings.Shape.Name=@n16.Value.Name
    Settings.Shape.Transform.Translation=@n16.Value.Transform.Translation
    Settings.Shape.bVisible=@n16.Value.bVisible
    Settings.PreferredRotationOrder=RigVMVariableNode() {variable=Rotation Order}.Value
    Settings.DisplayName=@n19.Value
   ) {template="SpawnControl::Execute(in InitialValue,in Settings,in OffsetTransform,in OffsetSpace,in Parent,in Name,out Item)"}
   > RigVMVariableNode(Value=@n20.Item) {variable=End Control}
   n21 = RigVMVariableNode() {variable=Bones}
   > Set_Metadata_1 = RigVMDispatchNode(
    Name=Offset
    NameSpace=Self
    Value=&RigUnit_GetRelativeTransformForItem(
     bChildInitial=true
     bParentInitial=true
     Parent=RigVMDispatchNode(Index=-1, Array=@n21.Value) {template="DISPATCH_RigVMDispatch_ArrayGetAtIndex(in Array,in Index,out Element)"}.Element
     Child=@n20.Item
    ).RelativeTransform
    Item=@n20.Item
   ) {template="DISPATCH_RigDispatch_SetMetadata(in Item,in Name,in NameSpace,in Value,out Success)"}
   > Create Gimbal Control_1 = RigVMFunctionReferenceNode(
    Shape Settings Override=(bVisible=True,Name="None",Color=(R=1,G=0,B=0,A=1),Transform=(Rotation=(X=0,Y=0,Z=0,W=1),Translation=(X=0,Y=0,Z=0),Scale3D=(X=1,Y=1,Z=1)))
    Rotation Order=RigVMVariableNode() {variable=Rotation Order}.Value
    Display Name=@n19.Value
    Source Control=@n20.Item
   )
   > RigVMVariableNode(Value=@"Create Gimbal Control_1".Gimbal Control) {variable=End Driver}
   > RigVMDispatchNode(Element=@"Create Gimbal Control_1".Gimbal Control, Array=RigVMVariableNode() {variable=IK Gimbal Controls}.Value) {template="DISPATCH_RigVMDispatch_ArrayAdd(io Array,in Element,out Index)"}
   > n22 = &RigUnit_HierarchyAddNull(
    Name=Mid
    Space=GlobalSpace
    Transform=(Rotation=(X=0,Y=0,Z=0,W=1),Translation=(X=0,Y=0,Z=0),Scale3D=(X=1,Y=1,Z=1))
    Transform.Translation=/Script/RigVM.RigVMFunction_MathVectorLerp(T=0.5, A=RigVMVariableNode() {variable=Start Position}.Value, B=RigVMVariableNode() {variable=End Position}.Value) {is=struct RigVMUnitNode, template="Interpolate::Execute(in A,in B,in T,out Result)"}.Result
    Parent=RigVMVariableNode() {variable=End Parent}.Value
   )
   > RigVMVariableNode(Value=@n22.Item) {variable=Mid Null}
   n23 = RigVMVariableNode() {variable=IKs Control Shapes}
   Control Color Override From Metadata_4 = RigVMFunctionReferenceNode(Color Override=@n23.Value.Color)
   n24 = RigVMVariableNode() {variable=Control Scale Factor Profile}
   Multiply = /Script/RigVM.RigVMFunction_MathVectorMul(
    B=(X=0.8,Y=0.8,Z=0.8)
    A=&RigVMFunction_MathVectorScale(
     Value=@n23.Value.Transform.Scale3D
     Factor=&RigVMFunction_AnimEvalRichCurve(SourceMinimum=0, TargetMinimum=0, Value=0.5, Curve=@n24.Value).Result
    ) {template="Scale::Execute(in Value,in Factor,out Result)"}.Result
   ) {is=struct RigVMUnitNode, template="Multiply::Execute(in A,in B,out Result)"}
   n25 = RigVMVariableNode() {variable=Rotation Offset}
   > n26 = &RigUnit_HierarchyAddControlTransform(
    InitialValue=(Rotation=(X=0,Y=0,Z=0,W=1),Translation=(X=0,Y=0,Z=0),Scale3D=(X=1,Y=1,Z=1))
    Name=Mid
    OffsetSpace=LocalSpace
    OffsetTransform=(Rotation=(X=0,Y=0,Z=0,W=1),Translation=(X=0,Y=0,Z=0),Scale3D=(X=1,Y=1,Z=1))
    Settings="(InitialSpace=LocalSpace,bUsePreferredRotationOrder=true,PreferredRotationOrder=YZX,Limits=(LimitTranslationX=(bMinimum=False,bMaximum=False),LimitTranslationY=(bMinimum=False,bMaximum=False),LimitTranslationZ=(bMinimum=False,bMaximum=False),LimitPitch=(bM ...(truncated)"
    Settings.Shape.Color=@"Control Color Override From Metadata_4".Result
    Settings.Shape.Transform.Scale3D=@Multiply.Result
    Settings.Shape.Transform.Rotation=&RigVMFunction_MathQuaternionMul(B=@n25.Value, A=@n23.Value.Transform.Rotation) {template="Multiply::Execute(in A,in B,out Result)"}.Result
    Settings.DisplayName=RigVMVariableNode() {variable=IK Mid Display Name}.Value
    Settings.PreferredRotationOrder=RigVMVariableNode() {variable=Rotation Order}.Value
    OffsetTransform.Rotation=@n25.Value
    Settings.Shape.Name=@n23.Value.Name
    Settings.Shape.Transform.Translation=@n23.Value.Transform.Translation
    Settings.Shape.bVisible=@n23.Value.bVisible
    Parent=@n22.Item
   ) {template="SpawnControl::Execute(in InitialValue,in Settings,in OffsetTransform,in OffsetSpace,in Parent,in Name,out Item)"}
   > RigVMVariableNode(Value=@n26.Item) {variable=Mid Control}
   > SpawnAnimationChannel = /Script/ControlRig.RigUnit_HierarchyAddAnimationChannelBool(InitialValue=true, LimitsEnabled=(), MaximumValue=True, MinimumValue=False, Name=Smarter Match, Parent=@n26.Item) {
    is=struct RigVMUnitNode
    template="SpawnAnimationChannel::Execute(in InitialValue,in MinimumValue,in MaximumValue,in LimitsEnabled,in Parent,in Name,out Item)"
   }
   > RigVMVariableNode(Value=@SpawnAnimationChannel.Item) {variable=Mid Smart Match Channel}
   n27 = RigVMVariableNode() {variable=Bones}
   n28 = /Script/RigVM.RigVMFunction_MathIntDiv(B=2, A=RigVMDispatchNode(Array=@n27.Value) {template="DISPATCH_RigVMDispatch_ArrayGetNum(in Array,out Num)"}.Num) {is=struct RigVMUnitNode, template="Divide::Execute(in A,in B,out Result)"}
   > Set_Metadata_1_1 = RigVMDispatchNode(
    Name=Offset
    NameSpace=Self
    Value=&RigUnit_GetRelativeTransformForItem(
     bChildInitial=true
     bParentInitial=true
     Parent=RigVMDispatchNode(Index=@n28.Result, Array=@n27.Value) {template="DISPATCH_RigVMDispatch_ArrayGetAtIndex(in Array,in Index,out Element)"}.Element
     Child=@n26.Item
    ).RelativeTransform
    Item=@n26.Item
   ) {template="DISPATCH_RigDispatch_SetMetadata(in Item,in Name,in NameSpace,in Value,out Success)"}
   RigVMFunction_MathBoolNot = &RigVMFunction_MathBoolNot(Value=RigVMVariableNode() {variable=Is Neck}.Value)
   n29 = &RigVMFunction_ControlFlowBranch(Condition=@RigVMFunction_MathBoolNot.Result, ExecuteContext=@Set_Metadata_1_1.ExecuteContext)
   Construct Movable Proxy v01 = RigVMFunctionReferenceNode(
    Proxy Name=End Movable Pivot
    Proxy Shape Settings="(DisplayName=\"None\",InitialSpace=LocalSpace,bUsePreferredRotationOrder=False,PreferredRotationOrder=YZX,Limits=(LimitTranslationX=(bMinimum=False,bMaximum=False),LimitTranslationY=(bMinimum=False,bMaximum=False),LimitTranslationZ=(bMinimum=False,bMaximum=F ...(truncated)"
    Proxy Vis=false
    Proxy Snap To=RigVMDispatchNode(Index=-1, Array=RigVMVariableNode() {variable=Bones}.Value) {template="DISPATCH_RigVMDispatch_ArrayGetAtIndex(in Array,in Index,out Element)"}.Element
    Proxy Parent Control=RigVMDispatchNode(Index=0, Array=RigVMVariableNode() {variable=FK Controls}.Value) {template="DISPATCH_RigVMDispatch_ArrayGetAtIndex(in Array,in Index,out Element)"}.Element
    ExecuteContext=@n29.True
    Proxy Shape Settings.DisplayName=RigVMVariableNode() {variable=End Moveable Pivot Display Name}.Value
    Driven Control=RigVMVariableNode() {variable=End Control}.Value
   )
   > RigVMVariableNode(Value=@"Construct Movable Proxy v01".Proxy Control) {variable=End Proxy Control}
   > RigVMVariableNode(Value=@"Construct Movable Proxy v01".Proxy Buffer) {variable=End Proxy Buffer}
   > RigVMVariableNode(Value=@"Construct Movable Proxy v01".Proxy Vis Channel) {variable=End Proxy Vis Channel}
   RigVMDispatchNode(Array=RigVMVariableNode() {variable=Sec FK Controls}.Value, ExecuteContext=@RigVMFunction_Sequence.C) {template="DISPATCH_RigVMDispatch_ArrayReset(io Array)"}
   n30 = /Script/ControlRig.RigUnit_HierarchyGetParent(
    bDefaultParent=True
    Child=RigVMDispatchNode(Index=0, Array=RigVMVariableNode() {variable=Bones}.Value) {template="DISPATCH_RigVMDispatch_ArrayGetAtIndex(in Array,in Index,out Element)"}.Element
   ) {is=struct RigVMUnitNode}
   > RigVMVariableNode(Value=@n30.Parent) {variable=Sec FK Parent}
   > RigVMVariableNode(Value=@n30.Parent) {variable=Match Null Parent}
   > RigVMDispatchNode(Array=RigVMVariableNode() {variable=Match Nulls}.Value) {template="DISPATCH_RigVMDispatch_ArrayReset(io Array)"}
   > For_Each_1 = RigVMDispatchNode(Array=RigVMVariableNode() {variable=Bones}.Value) {template="DISPATCH_RigVMDispatch_ArrayIterator(in Array,out Element,out Index,out Count,out Ratio)"}
   n31 = RigVMVariableNode() {variable=Control Names As FK}
   > Get Item Name v01_1 = RigVMFunctionReferenceNode(
    Index=0
    Numerical Name Start Index=1
    Remove Side Suffix=true
    Count=@For_Each_1.Count
    Name=@For_Each_1.Element.Name
    Prefix=RigVMDispatchNode(True=Sec FK, Condition=@n31.Value) {template="DISPATCH_RigVMDispatch_If(in Condition,in True,in False,out Result)"}.Result
    Suffix=RigVMDispatchNode(False=FK, Condition=@n31.Value) {template="DISPATCH_RigVMDispatch_If(in Condition,in True,in False,out Result)"}.Result
    Standard Numerical Names=@n31.Value
   )
   n32 = RigVMVariableNode() {variable="Sec FKs Control Shapes "}
   Control Color Override From Metadata_5 = RigVMFunctionReferenceNode(Color Override=@n32.Value.Color)
   Get Optional Display Name = RigVMFunctionReferenceNode(Array=RigVMVariableNode() {variable=FK Secondary Display Names}.Value, Index=@For_Each_1.Index)
   n33 = &RigUnit_GetTransform(bInitial=true, Item=@For_Each_1.Element)
   Make Transform = RigVMDispatchNode(
    Value=(Rotation=(X=0,Y=0,Z=0,W=1),Translation=(X=0,Y=0,Z=0),Scale3D=(X=1,Y=1,Z=1))
    Value.Rotation=@n33.Transform.Rotation
    Value.Translation=@n33.Transform.Translation
   ) {template="DISPATCH_RigVMDispatch_Constant(io Value)"}
   > n34 = &RigUnit_HierarchyAddControlTransform(
    InitialValue=(Rotation=(X=0,Y=0,Z=0,W=1),Translation=(X=0,Y=0,Z=0),Scale3D=(X=1,Y=1,Z=1))
    OffsetSpace=GlobalSpace
    Settings="(InitialSpace=LocalSpace,bUsePreferredRotationOrder=true,PreferredRotationOrder=YZX,Limits=(LimitTranslationX=(bMinimum=False,bMaximum=False),LimitTranslationY=(bMinimum=False,bMaximum=False),LimitTranslationZ=(bMinimum=False,bMaximum=False),LimitPitch=(bM ...(truncated)"
    Settings.Shape.Color=@"Control Color Override From Metadata_5".Result
    Name=@"Get Item Name v01_1".Result
    Settings.DisplayName=@"Get Optional Display Name".Display Name
    OffsetTransform=@"Make Transform".Value
    Settings.PreferredRotationOrder=RigVMVariableNode() {variable=Rotation Order}.Value
    Settings.Shape.Name=@n32.Value.Name
    Settings.Shape.Transform=@n32.Value.Transform
    Settings.Shape.bVisible=@n32.Value.bVisible
    Parent=RigVMVariableNode() {variable=Sec FK Parent}.Value
   ) {template="SpawnControl::Execute(in InitialValue,in Settings,in OffsetTransform,in OffsetSpace,in Parent,in Name,out Item)"}
   > RigVMVariableNode(Value=@n34.Item) {variable=Sec FK Parent}
   > RigVMDispatchNode(Array=RigVMVariableNode() {variable=Sec FK Controls}.Value, Element=@n34.Item) {template="DISPATCH_RigVMDispatch_ArrayAdd(io Array,in Element,out Index)"}
   > n35 = &RigUnit_HierarchyAddNull(
    Space=GlobalSpace
    Name=&RigVMFunction_NameConcat(B=_null, A=@For_Each_1.Element.Name) {template="Concat::Execute(in A,in B,out Result)"}.Result
    Transform=@"Make Transform".Value
    Parent=RigVMVariableNode() {variable=Match Null Parent}.Value
   )
   > RigVMVariableNode(Value=@n35.Item) {variable=Match Null Parent}
   > RigVMDispatchNode(Element=@n35.Item, Array=RigVMVariableNode() {variable=Match Nulls}.Value) {template="DISPATCH_RigVMDispatch_ArrayAdd(io Array,in Element,out Index)"}
   RigVMDispatchNode(Array=RigVMVariableNode() {variable=Start IK Space Nulls}.Value, ExecuteContext=@RigVMFunction_Sequence.F) {template="DISPATCH_RigVMDispatch_ArrayReset(io Array)"}
   > Constuct Space Nulls v01_2 = RigVMFunctionReferenceNode(
    Orient Only=true
    Driven Control=RigVMVariableNode() {variable=Start Control}.Value
    Space Null Parent=RigVMVariableNode() {variable=Start Parent}.Value
    Spaces=RigVMVariableNode() {variable=Start Spaces}.Value
   )
   > RigVMDispatchNode(Other=@"Constuct Space Nulls v01_2".Output Space Nulls, Array=RigVMVariableNode() {variable=Start IK Space Nulls}.Value) {template="DISPATCH_RigVMDispatch_ArrayAppend(io Array,in Other)"}
   > RigVMDispatchNode(Array=RigVMVariableNode() {variable=Start FK Space Nulls}.Value) {template="DISPATCH_RigVMDispatch_ArrayReset(io Array)"}
   n36 = RigVMDispatchNode(Index=0, Array=RigVMVariableNode() {variable=FK Controls}.Value) {template="DISPATCH_RigVMDispatch_ArrayGetAtIndex(in Array,in Index,out Element)"}
   > Constuct Space Nulls v01_3 = RigVMFunctionReferenceNode(
    Orient Only=true
    Driven Control=@n36.Element
    Space Null Parent=&RigUnit_HierarchyGetParent(bDefaultParent=True, Child=&RigUnit_HierarchyGetParent(bDefaultParent=True, Child=@n36.Element).Parent).Parent
    Spaces=RigVMVariableNode() {variable=Start Fk Spaces}.Value
   )
   > RigVMDispatchNode(Other=@"Constuct Space Nulls v01_3".Output Space Nulls, Array=RigVMVariableNode() {variable=Start FK Space Nulls}.Value) {template="DISPATCH_RigVMDispatch_ArrayAppend(io Array,in Other)"}
   RigVMDispatchNode(Array=RigVMVariableNode() {variable=End IK Space Nulls}.Value, ExecuteContext=@RigVMFunction_Sequence.G) {template="DISPATCH_RigVMDispatch_ArrayReset(io Array)"}
   > Constuct Space Nulls v01 = RigVMFunctionReferenceNode(
    Orient Only=true
    Driven Control=RigVMVariableNode() {variable=End Control}.Value
    Space Null Parent=RigVMVariableNode() {variable=End Parent}.Value
    Spaces=RigVMVariableNode() {variable=End Spaces}.Value
   )
   > RigVMDispatchNode(Other=@"Constuct Space Nulls v01".Output Space Nulls, Array=RigVMVariableNode() {variable=End IK Space Nulls}.Value) {template="DISPATCH_RigVMDispatch_ArrayAppend(io Array,in Other)"}
   > RigVMDispatchNode(Array=RigVMVariableNode() {variable=End FK Space Nulls}.Value) {template="DISPATCH_RigVMDispatch_ArrayReset(io Array)"}
   n37 = RigVMDispatchNode(Index=-1, Array=RigVMVariableNode() {variable=FK Controls}.Value) {template="DISPATCH_RigVMDispatch_ArrayGetAtIndex(in Array,in Index,out Element)"}
   > Constuct Space Nulls v01_1 = RigVMFunctionReferenceNode(
    Orient Only=true
    Driven Control=@n37.Element
    Space Null Parent=&RigUnit_HierarchyGetParent(bDefaultParent=True, Child=&RigUnit_HierarchyGetParent(bDefaultParent=True, Child=@n37.Element).Parent).Parent
    Spaces=RigVMVariableNode() {variable=End Fk Spaces}.Value
   )
   > RigVMDispatchNode(Other=@"Constuct Space Nulls v01_1".Output Space Nulls, Array=RigVMVariableNode() {variable=End FK Space Nulls}.Value) {template="DISPATCH_RigVMDispatch_ArrayAppend(io Array,in Other)"}
   Get Chain Length = RigVMFunctionReferenceNode(
    Initial=true
    Items=((Type=Bone,Name="torso_1"),(Type=Bone,Name="torso_2"))
    Items.0=RigVMVariableNode() {variable=Start Driver}.Value
    Items.1=RigVMVariableNode() {variable=End Driver}.Value
    ExecuteContext=@RigVMFunction_Sequence.H
   )
   Clone = RigVMDispatchNode(Array=RigVMVariableNode() {variable=FK Controls}.Value) {template="DISPATCH_RigVMDispatch_ArrayClone(in Array,out Clone)"}
   > n38 = RigVMDispatchNode(Array=@Clone.Clone, Other=RigVMVariableNode() {variable=Sec FK Controls}.Value) {template="DISPATCH_RigVMDispatch_ArrayUnion(io Array,in Other)"}
   > n39 = RigVMDispatchNode(Array=@n38.Array, Other=RigVMVariableNode() {variable=FK Gimbal Controls}.Value) {template="DISPATCH_RigVMDispatch_ArrayUnion(io Array,in Other)"}
   > n40 = RigVMDispatchNode(Array=@n39.Array, Other=RigVMVariableNode() {variable=IK Gimbal Controls}.Value) {template="DISPATCH_RigVMDispatch_ArrayUnion(io Array,in Other)"}
   > n41 = RigVMDispatchNode(
    Other=&RigUnit_ItemArray(
     Items=((Type=None,Name="None"),(Type=None,Name="None"),(Type=None,Name="None"))
     Items.0=RigVMVariableNode() {variable=Start Control}.Value
     Items.1=RigVMVariableNode() {variable=End Control}.Value
     Items.2=RigVMVariableNode() {variable=Mid Control}.Value
    ).Items
    Array=@n40.Array
   ) {template="DISPATCH_RigVMDispatch_ArrayUnion(io Array,in Other)"}
   Get_Module_Metadata_1 = RigVMDispatchNode(Default=1, Name=Global Control Scale, NameSpace=Root) {template="DISPATCH_RigDispatch_GetModuleMetadata(in Name,in NameSpace,in Default,out Value,out Found)"}
   > Set Control Scale = RigVMFunctionReferenceNode(
    Auto Scale Factor=&RigVMFunction_MathDoubleDiv(B=30, A=@"Get Chain Length".Length).Result
    Global Scale=@Get_Module_Metadata_1.Value
    Controls=@n41.Array
    Scale=RigVMVariableNode() {variable=Control Scale}.Value
   )
   n42 = RigVMVariableNode() {variable=End Control}
   SpawnScaleAnimationChannel = /Script/ControlRig.RigUnit_HierarchyAddAnimationChannelScaleFloat(
    InitialValue=1
    LimitsEnabled=(Enabled=(bMinimum=True,bMaximum=True))
    MaximumValue=1
    MinimumValue=0
    Name=Stretch
    Parent=@n42.Value
    ExecutePin=@RigVMFunction_Sequence.D
   ) {
    is=struct RigVMUnitNode
    template="SpawnScaleAnimationChannel::Execute(in InitialValue,in MinimumValue,in MaximumValue,in LimitsEnabled,in Parent,in Name,out Item)"
   }
   > RigVMVariableNode(Value=@SpawnScaleAnimationChannel.Item) {variable=Stretch Channel}
   > &RigUnit_HierarchyAddAnimationChannelScaleFloat(
    LimitsEnabled=(Enabled=(bMinimum=True,bMaximum=True))
    MaximumValue=1
    MinimumValue=0
    Name=Distribute Rotation
    InitialValue=RigVMVariableNode() {variable=Default Distribute Rotation}.Value
    Parent=@n42.Value
   ) {template="SpawnScaleAnimationChannel::Execute(in InitialValue,in MinimumValue,in MaximumValue,in LimitsEnabled,in Parent,in Name,out Item)"}
   > &RigUnit_HierarchyAddAnimationChannelScaleFloat(
    InitialValue=1
    LimitsEnabled=(Enabled=(bMinimum=True,bMaximum=True))
    MaximumValue=1
    MinimumValue=0
    Name=In Between
    Parent=RigVMVariableNode() {variable=Mid Control}.Value
   ) {template="SpawnScaleAnimationChannel::Execute(in InitialValue,in MinimumValue,in MaximumValue,in LimitsEnabled,in Parent,in Name,out Item)"}
   > &RigUnit_HierarchyAddAnimationChannelScaleFloat(
    InitialValue=0.65
    LimitsEnabled=(Enabled=(bMinimum=True,bMaximum=True))
    MaximumValue=1
    MinimumValue=0
    Name=Pelvis Chest
    Parent=RigVMVariableNode() {variable=Mid Control}.Value
   ) {template="SpawnScaleAnimationChannel::Execute(in InitialValue,in MinimumValue,in MaximumValue,in LimitsEnabled,in Parent,in Name,out Item)"}
   > /Script/ControlRig.RigUnit_SetChannelHosts(
    Hosts=((Type=None,Name="None"),(Type=None,Name="None"),(Type=None,Name="None"))
    Hosts.0=RigVMDispatchNode(Index=0, Array=RigVMVariableNode() {variable=FK Controls}.Value) {template="DISPATCH_RigVMDispatch_ArrayGetAtIndex(in Array,in Index,out Element)"}.Element
    Hosts.1=RigVMDispatchNode(Index=-1, Array=RigVMVariableNode() {variable=FK Controls}.Value) {template="DISPATCH_RigVMDispatch_ArrayGetAtIndex(in Array,in Index,out Element)"}.Element
    Channel=RigVMVariableNode() {variable=Stretch Channel}.Value
    Hosts.2=RigVMVariableNode() {variable=Start Control}.Value
   ) {is=struct RigVMUnitNode}
   n43 = RigVMVariableNode() {variable=Start Control}
   > n44 = &RigUnit_HierarchyAddAnimationChannelBool(
    LimitsEnabled=()
    MaximumValue=True
    MinimumValue=False
    Name=IK Vis
    InitialValue=&RigVMFunction_MathBoolNot(Value=RigVMVariableNode() {variable=Is Neck}.Value).Result
    Parent=@n43.Value
   ) {template="SpawnAnimationChannel::Execute(in InitialValue,in MinimumValue,in MaximumValue,in LimitsEnabled,in Parent,in Name,out Item)"}
   > RigVMVariableNode(Value=@n44.Item) {variable=IKs Visibility Channel}
   > n45 = &RigUnit_HierarchyAddAnimationChannelBool(InitialValue=true, LimitsEnabled=(), MaximumValue=True, MinimumValue=False, Name=FK Vis, Parent=@n43.Value) {template="SpawnAnimationChannel::Execute(in InitialValue,in MinimumValue,in MaximumValue,in LimitsEnabled,in Parent,in Name,out Item)"}
   > RigVMVariableNode(Value=@n45.Item) {variable=FKs Visibility Channel}
   > n46 = &RigUnit_HierarchyAddAnimationChannelBool(InitialValue=false, LimitsEnabled=(), MaximumValue=True, MinimumValue=False, Name=Sec FK Vis, Parent=@n43.Value) {template="SpawnAnimationChannel::Execute(in InitialValue,in MinimumValue,in MaximumValue,in LimitsEnabled,in Parent,in Name,out Item)"}
   > RigVMVariableNode(Value=@n46.Item) {variable=Sec FKs Visibility Channel}
   > For_Each_2 = RigVMDispatchNode(
    Array=((Type=None,Name="None"),(Type=None,Name="None"),(Type=None,Name="None"))
    Array.1=RigVMVariableNode() {variable=FKs Visibility Channel}.Value
    Array.2=RigVMVariableNode() {variable=IKs Visibility Channel}.Value
    Array.0=@n46.Item
   ) {template="DISPATCH_RigVMDispatch_ArrayIterator(in Array,out Element,out Index,out Count,out Ratio)"}
   > &RigUnit_SetChannelHosts(
    Hosts=((Type=None,Name="None"),(Type=None,Name="None"),(Type=None,Name="None"),(Type=None,Name="None"),(Type=None,Name="None"))
    Hosts.4=RigVMDispatchNode(Index=-1, Array=RigVMVariableNode() {variable=Sec FK Controls}.Value) {template="DISPATCH_RigVMDispatch_ArrayGetAtIndex(in Array,in Index,out Element)"}.Element
    Hosts.0=RigVMDispatchNode(Index=0, Array=RigVMVariableNode() {variable=FK Controls}.Value) {template="DISPATCH_RigVMDispatch_ArrayGetAtIndex(in Array,in Index,out Element)"}.Element
    Hosts.1=RigVMDispatchNode(Index=-1, Array=RigVMVariableNode() {variable=FK Controls}.Value) {template="DISPATCH_RigVMDispatch_ArrayGetAtIndex(in Array,in Index,out Element)"}.Element
    Hosts.3=RigVMDispatchNode(Index=0, Array=RigVMVariableNode() {variable=Sec FK Controls}.Value) {template="DISPATCH_RigVMDispatch_ArrayGetAtIndex(in Array,in Index,out Element)"}.Element
    Channel=@For_Each_2.Element
    Hosts.2=RigVMVariableNode() {variable=End Control}.Value
   )
   n47 = &RigUnit_GetTransform(bInitial=true, Item=RigVMVariableNode() {variable=End Driver}.Value)
   RigVMFunction_MathMatrixFromTransformV2 = /Script/RigVM.RigVMFunction_MathMatrixFromTransformV2(Value=@n47.Transform) {is=struct RigVMUnitNode, template="Cast::Execute(in Value,out Result)"}
   RigVMFunction_MathMatrixToVectors = /Script/RigVM.RigVMFunction_MathMatrixToVectors(Value=@RigVMFunction_MathMatrixFromTransformV2.Result) {is=struct RigVMUnitNode}
   n48 = &RigUnit_GetTransform(bInitial=true, Item=RigVMVariableNode() {variable=Start Driver}.Value)
   Interpolate = &RigVMFunction_MathVectorLerp(
    T=0
    B=@RigVMFunction_MathMatrixToVectors.X
    A=&RigVMFunction_MathMatrixToVectors(Value=&RigVMFunction_MathMatrixFromTransformV2(Value=@n48.Transform) {template="Cast::Execute(in Value,out Result)"}.Result).X
   ) {template="Interpolate::Execute(in A,in B,in T,out Result)"}
   Subtract = /Script/RigVM.RigVMFunction_MathVectorSub(B=@n48.Transform.Translation, A=@n47.Transform.Translation) {is=struct RigVMUnitNode, template="Subtract::Execute(in A,in B,out Result)"}
   RigVMFunction_MathVectorCross = /Script/RigVM.RigVMFunction_MathVectorCross(B=@Interpolate.Result, A=@Subtract.Result) {is=struct RigVMUnitNode, template="Cross::Execute(in A,in B,out Result)"}
   RigVMVariableNode(Value=@RigVMFunction_MathVectorCross.Result, ExecuteContext=@RigVMFunction_Sequence.I) {variable=Up Vector Axis}
   RigVMDispatchNode(Array=RigVMVariableNode() {variable=FK Controls}.Value, ExecuteContext=@RigVMFunction_Sequence.A) {template="DISPATCH_RigVMDispatch_ArrayReset(io Array)"}
   > RigVMDispatchNode(Array=RigVMVariableNode() {variable=FK Gimbal Controls}.Value) {template="DISPATCH_RigVMDispatch_ArrayReset(io Array)"}
   n49 = RigVMVariableNode() {variable=Bones}
   Num = RigVMDispatchNode(Array=@n49.Value) {template="DISPATCH_RigVMDispatch_ArrayGetNum(in Array,out Num)"}
   > RigVMVariableNode(
    Value=((Type=None,Name="None"),(Type=None,Name="None"),(Type=None,Name="None"))
    Value.0=RigVMDispatchNode(Index=0, Array=@n49.Value) {template="DISPATCH_RigVMDispatch_ArrayGetAtIndex(in Array,in Index,out Element)"}.Element
    Value.1=RigVMDispatchNode(Index=&RigVMFunction_MathIntDiv(B=2, A=@Num.Num) {template="Divide::Execute(in A,in B,out Result)"}.Result, Array=@n49.Value) {template="DISPATCH_RigVMDispatch_ArrayGetAtIndex(in Array,in Index,out Element)"}.Element
    Value.2=RigVMDispatchNode(Index=-1, Array=@n49.Value) {template="DISPATCH_RigVMDispatch_ArrayGetAtIndex(in Array,in Index,out Element)"}.Element
   ) {variable=FK Bones}
   > RigVMVariableNode(Value=RigVMVariableNode() {variable=Start Parent}.Value) {variable=FK Parent}
   > For_Each = RigVMDispatchNode(Array=RigVMVariableNode() {variable=FK Bones}.Value) {template="DISPATCH_RigVMDispatch_ArrayIterator(in Array,out Element,out Index,out Count,out Ratio)"}
   > Get Item Name v01 = RigVMFunctionReferenceNode(
    Numerical Name Start Index=1
    Prefix=FK
    Standard Numerical Names=true
    Count=@For_Each.Count
    Name=@For_Each.Element.Name
    Index=@For_Each.Index
   )
   Concat = RigVMAggregateNode(A="Default ", C=" Space", B=@"Get Item Name v01".Result)
   n50 = &RigUnit_GetTransform(bInitial=true, Item=@For_Each.Element)
   n51 = RigVMVariableNode() {variable=Rotation Offset}
   > n52 = &RigUnit_HierarchyAddNull(
    Space=GlobalSpace
    Transform=(Rotation=(X=0,Y=0,Z=0,W=1),Translation=(X=0,Y=0,Z=0),Scale3D=(X=1,Y=1,Z=1))
    Name=@Concat.Result
    Transform.Translation=@n50.Transform.Translation
    Parent=RigVMVariableNode() {variable=FK Parent}.Value
    Transform.Rotation=@n51.Value
   )
   n53 = RigVMVariableNode() {variable=FKs Control Shapes}
   Control Color Override From Metadata = RigVMFunctionReferenceNode(Color Override=@n53.Value.Color)
   RigVMFunction_AnimEvalRichCurve = &RigVMFunction_AnimEvalRichCurve(SourceMinimum=0, TargetMinimum=0, Value=@For_Each.Ratio, Curve=RigVMVariableNode() {variable=Control Scale Factor Profile}.Value)
   Scale = &RigVMFunction_MathVectorScale(Factor=@RigVMFunction_AnimEvalRichCurve.Result, Value=@n53.Value.Transform.Scale3D) {template="Scale::Execute(in Value,in Factor,out Result)"}
   Get Optional Display Name_1 = RigVMFunctionReferenceNode(Index=@For_Each.Index, Array=RigVMVariableNode() {variable=FK Primary Display Names}.Value)
   > SpawnControl = &RigUnit_HierarchyAddControlTransform(
    InitialValue=(Rotation=(X=0,Y=0,Z=0,W=1),Translation=(X=0,Y=0,Z=0),Scale3D=(X=1,Y=1,Z=1))
    OffsetSpace=GlobalSpace
    OffsetTransform=(Rotation=(X=0,Y=0,Z=0,W=1),Translation=(X=0,Y=0,Z=0),Scale3D=(X=1,Y=1,Z=1))
    Settings="(InitialSpace=LocalSpace,bUsePreferredRotationOrder=true,PreferredRotationOrder=YZX,Limits=(LimitTranslationX=(bMinimum=False,bMaximum=False),LimitTranslationY=(bMinimum=False,bMaximum=False),LimitTranslationZ=(bMinimum=False,bMaximum=False),LimitPitch=(bM ...(truncated)"
    Settings.Shape.Color=@"Control Color Override From Metadata".Result
    Name=@"Get Item Name v01".Result
    OffsetTransform.Translation=@n50.Transform.Translation
    Parent=@n52.Item
    Settings.Shape.Transform.Rotation=&RigVMFunction_MathQuaternionMul(B=@n51.Value, A=@n53.Value.Transform.Rotation) {template="Multiply::Execute(in A,in B,out Result)"}.Result
    Settings.Shape.Transform.Scale3D=@Scale.Result
    Settings.Shape.Name=@n53.Value.Name
    Settings.Shape.Transform.Translation=@n53.Value.Transform.Translation
    Settings.Shape.bVisible=@n53.Value.bVisible
    Settings.PreferredRotationOrder=RigVMVariableNode() {variable=Rotation Order}.Value
    Settings.DisplayName=@"Get Optional Display Name_1".Display Name
    OffsetTransform.Rotation=@n51.Value
   ) {template="SpawnControl::Execute(in InitialValue,in Settings,in OffsetTransform,in OffsetSpace,in Parent,in Name,out Item)"}
   > RigVMVariableNode(Value=@SpawnControl.Item) {variable=FK Parent}
   > RigVMVariableNode(Value=@SpawnControl.Item) {variable=End Parent}
   > Set_Metadata_2 = RigVMDispatchNode(
    Name=FkDeltaTransform
    NameSpace=Self
    Value=&RigVMFunction_MathTransformMakeRelative(Parent=@n50.Transform, Global=&RigUnit_GetTransform(bInitial=true, Item=@SpawnControl.Item).Transform).Local
    Item=@SpawnControl.Item
   ) {template="DISPATCH_RigDispatch_SetMetadata(in Item,in Name,in NameSpace,in Value,out Success)"}
   > Add = RigVMDispatchNode(Array=RigVMVariableNode() {variable=FK Controls}.Value, Element=@SpawnControl.Item) {template="DISPATCH_RigVMDispatch_ArrayAdd(io Array,in Element,out Index)"}
   RigVMFunction_ControlFlowBranch = &RigVMFunction_ControlFlowBranch(
    ExecuteContext=@Add.ExecuteContext
    Condition=RigVMDispatchNode(
     B=/Script/RigVM.RigVMFunction_MathIntAdd(B=1, A=@For_Each.Index) {is=struct RigVMUnitNode, template="Add::Execute(in A,in B,out Result)"}.Result
     A=@For_Each.Count
    ) {template="DISPATCH_RigVMDispatch_CoreEquals(in A,in B,out Result)"}.Result
   )
   Create Gimbal Control_2 = RigVMFunctionReferenceNode(
    Shape Settings Override=(bVisible=true,Name="None",Color=(R=1,G=0,B=0,A=1),Transform=(Rotation=(X=0,Y=0,Z=0,W=1),Translation=(X=0,Y=0,Z=0),Scale3D=(X=0.8,Y=0.8,Z=0.8)))
    ExecuteContext=@RigVMFunction_ControlFlowBranch.True
    Rotation Order=RigVMVariableNode() {variable=Rotation Order}.Value
    Display Name=@"Get Optional Display Name_1".Display Name
    Source Control=@SpawnControl.Item
   )
   > RigVMVariableNode(Value=@"Create Gimbal Control_2".Gimbal Control) {variable=FK Parent}
   > RigVMVariableNode(Value=@"Create Gimbal Control_2".Gimbal Control) {variable=End Parent}
   > RigVMDispatchNode(Array=RigVMVariableNode() {variable=FK Gimbal Controls}.Value, Element=@"Create Gimbal Control_2".Gimbal Control) {template="DISPATCH_RigVMDispatch_ArrayAdd(io Array,in Element,out Index)"}
   RigVMVariableNode() {variable=Start Parent}
   RigVMVariableNode() {variable=Start Control}
   RigVMVariableNode() {variable=End Control}
   RigVMVariableNode() {variable=Start Parent}
   RigVMVariableNode() {variable=Start Control}
   RigVMVariableNode() {variable=End Control}
  ]}
 ContainedGraph {nodes [
   n1 = RigVMFunctionEntryNode()
   n2 = &RigUnit_GetTransform(bInitial=true, Item=RigVMVariableNode() {variable=Proxy Snap To}.Value)
   > n3 = &RigUnit_HierarchyAddNull(
    Space=GlobalSpace
    Transform=(Rotation=(X=0,Y=0,Z=0,W=1),Translation=(X=0,Y=0,Z=0),Scale3D=(X=1,Y=1,Z=1))
    Name=&RigVMFunction_NameConcat(B=" Null", A=RigVMVariableNode() {variable=Proxy Name}.Value) {template="Concat::Execute(in A,in B,out Result)"}.Result
    Parent=@n1.Proxy Parent Control
    Transform.Rotation=&RigUnit_GetTransform(bInitial=true, Item=RigVMVariableNode() {variable=Driven Control}.Value).Transform.Rotation
    Transform.Scale3D=@n2.Transform.Scale3D
    Transform.Translation=@n2.Transform.Translation
   )
   > SpawnControl = &RigUnit_HierarchyAddControlTransform(
    InitialValue=(Rotation=(X=0,Y=0,Z=0,W=1),Translation=(X=0,Y=0,Z=0),Scale3D=(X=1,Y=1,Z=1))
    OffsetSpace=LocalSpace
    OffsetTransform=(Rotation=(X=0,Y=0,Z=0,W=1),Translation=(X=0,Y=0,Z=0),Scale3D=(X=1,Y=1,Z=1))
    Parent=@n3.Item
    Name=RigVMVariableNode() {variable=Proxy Name}.Value
    Settings=RigVMVariableNode() {variable=Proxy Shape Settings}.Value
   ) {template="SpawnControl::Execute(in InitialValue,in Settings,in OffsetTransform,in OffsetSpace,in Parent,in Name,out Item)"}
   > SpawnAnimationChannel = &RigUnit_HierarchyAddAnimationChannelBool(
    InitialValue=true
    LimitsEnabled=()
    MaximumValue=True
    MinimumValue=False
    Name=Movable Pivot Vis
    Parent=RigVMVariableNode() {variable=Driven Control}.Value
   ) {template="SpawnAnimationChannel::Execute(in InitialValue,in MinimumValue,in MaximumValue,in LimitsEnabled,in Parent,in Name,out Item)"}
   > n4 = /Script/ControlRig.RigUnit_HierarchyAddAnimationChannelFloat(
    InitialValue=0.5
    LimitsEnabled=(Enabled=(bMinimum=True,bMaximum=True))
    MaximumValue=1
    MinimumValue=0
    Name=Proxy Pivot Slide
    Parent=@SpawnControl.Item
   ) {
    is=struct RigVMUnitNode
    template="SpawnAnimationChannel::Execute(in InitialValue,in MinimumValue,in MaximumValue,in LimitsEnabled,in Parent,in Name,out Item)"
   }
   > SetControlDrivenList = /Script/ControlRig.RigUnit_SetControlDrivenList(Driven=((Type=None,Name="None")), Control=@SpawnControl.Item.Name, Driven.0=RigVMVariableNode() {variable=Driven Control}.Value) {is=struct RigVMUnitNode, template="SetControlDrivenList::Execute(in Control,in Driven)"}
   Concat = &RigVMFunction_NameConcat(B=" Bfr", A=RigVMVariableNode() {variable=Proxy Name}.Value) {template="Concat::Execute(in A,in B,out Result)"}
   > n5 = &RigUnit_HierarchyAddNull(
    Space=GlobalSpace
    Transform=(Rotation=(X=0,Y=0,Z=0,W=1),Translation=(X=0,Y=0,Z=0),Scale3D=(X=1,Y=1,Z=1))
    Name=@Concat.Result
    Parent=@SpawnControl.Item
   )
   > RigVMDispatchNode(Name=IsSet, NameSpace=Self, Value=False, Item=@n5.Item) {template="DISPATCH_RigDispatch_SetMetadata(in Item,in Name,in NameSpace,in Value,out Success)"}
   > RigVMFunctionReturnNode(Proxy Buffer=@n5.Item, Proxy Vis Channel=@SpawnAnimationChannel.Item, Proxy Control=@SpawnControl.Item, Proxy Slide Channel=@n4.Item)
  ]}
 ContainedGraph {nodes [
   RigVMFunctionEntryNode()
   > SequenceAggregate = RigVMAggregateNode()
   n1 = /Script/ControlRig.RigUnit_IsInteracting() {is=struct RigVMUnitNode}
   n2 = RigVMDispatchNode(Array=@n1.Items, Element=RigVMVariableNode() {variable=Proxy Control}.Value) {template="DISPATCH_RigVMDispatch_ArrayFind(in Array,in Element,out Index,out Success)"}
   RigVMFunction_MathBoolAnd = &RigVMFunction_MathBoolAnd(B=@n2.Success, A=@n1.bIsTranslating)
   RigVMFunction_MathBoolOr = &RigVMFunction_MathBoolOr(
    B=RigVMDispatchNode(Array=@n1.Items, Element=RigVMVariableNode() {variable=Proxy Slide Channel}.Value) {template="DISPATCH_RigVMDispatch_ArrayFind(in Array,in Element,out Index,out Success)"}.Success
    A=@RigVMFunction_MathBoolAnd.Result
   )
   RigVMFunction_ControlFlowBranch = &RigVMFunction_ControlFlowBranch(Condition=@RigVMFunction_MathBoolOr.Result, ExecuteContext=@SequenceAggregate.A)
   n3 = RigVMVariableNode() {variable=Proxy Buffer}
   &RigUnit_SetTransform(
    Weight=1
    bInitial=False
    Value=&RigUnit_GetTransform(Item=RigVMVariableNode() {variable=Driven Control}.Value).Transform
    ExecutePin=@RigVMFunction_ControlFlowBranch.True
    Item=@n3.Value
   )
   > RigVMDispatchNode(Name=IsSet, NameSpace=Self, Value=true, Item=@n3.Value) {template="DISPATCH_RigDispatch_SetMetadata(in Item,in Name,in NameSpace,in Value,out Success)"}
   n4 = &RigVMFunction_ControlFlowBranch(Condition=&RigVMFunction_MathBoolAnd(A=@n2.Success, B=@n1.bIsRotating).Result, ExecuteContext=@SequenceAggregate.B)
   RigVMDispatchNode(Name=IsSet, NameSpace=Self, Value=False, ExecuteContext=@n4.False, Item=RigVMVariableNode() {variable=Proxy Buffer}.Value) {template="DISPATCH_RigDispatch_SetMetadata(in Item,in Name,in NameSpace,in Value,out Success)"}
   GetAnimationChannelFromItem = /Script/ControlRig.RigUnit_GetFloatAnimationChannelFromItem(bInitial=False, Item=RigVMVariableNode() {variable=Proxy Slide Channel}.Value) {is=struct RigVMUnitNode, template="GetAnimationChannelFromItem::Execute(out Value,in Item,in bInitial)"}
   > &RigUnit_SetTransform(
    Value=(Rotation=(X=0,Y=0,Z=0,W=1),Translation=(X=0,Y=0,Z=0),Scale3D=(X=1,Y=1,Z=1))
    Weight=1
    bInitial=False
    Value.Rotation=&RigUnit_GetTransform(bInitial=False, Item=RigVMVariableNode() {variable=Driven Control}.Value).Transform.Rotation
    Item=&RigUnit_HierarchyGetParent(bDefaultParent=True, Child=RigVMVariableNode() {variable=Proxy Control}.Value).Parent
    Value.Translation=/Script/ControlRigSpline.RigUnit_PositionFromControlRigSpline(U=@GetAnimationChannelFromItem.Value, Spline=RigVMVariableNode() {variable=Spline}.Value) {is=struct RigVMUnitNode}.Position
   )
   > Set Transform_1 = &RigUnit_SetRotation(Value=(X=0,Y=0,Z=0,W=1), Weight=1, bInitial=False, Item=RigVMVariableNode() {variable=Proxy Control}.Value)
   SequenceAggregate_5_1 = &RigVMFunction_Sequence(ExecuteContext=@n4.True)
   RigVMFunction_MathBoolNot = &RigVMFunction_MathBoolNot(Value=RigVMDispatchNode(Default=False, Name=IsSet, NameSpace=Self, Item=RigVMVariableNode() {variable=Proxy Buffer}.Value) {template="DISPATCH_RigDispatch_GetMetadata(in Item,in Name,in NameSpace,in Default,out Value,out Found)"}.Value)
   n5 = &RigVMFunction_ControlFlowBranch(Condition=@RigVMFunction_MathBoolNot.Result, ExecuteContext=@SequenceAggregate_5_1.A)
   n6 = RigVMVariableNode() {variable=Proxy Buffer}
   &RigUnit_SetTransform(
    Weight=1
    bInitial=False
    ExecutePin=@n5.True
    Value=&RigUnit_GetTransform(Item=RigVMVariableNode() {variable=Driven Control}.Value).Transform
    Item=@n6.Value
   )
   > RigVMDispatchNode(Name=IsSet, NameSpace=Self, Value=true, Item=@n6.Value) {template="DISPATCH_RigDispatch_SetMetadata(in Item,in Name,in NameSpace,in Value,out Success)"}
   &RigUnit_SetTransform(
    Weight=1
    bInitial=False
    Value=&RigUnit_GetTransform(Item=RigVMVariableNode() {variable=Proxy Buffer}.Value).Transform
    ExecutePin=@SequenceAggregate_5_1.B
    Item=RigVMVariableNode() {variable=Driven Control}.Value
   )
   RigVMFunctionReturnNode(ExecuteContext=@SequenceAggregate.C)
  ]}
 New Function_ContainedGraph {nodes [
   n1 = RigVMFunctionEntryNode()
   n2 = &RigUnit_GetTransformItemArray(
    Items=((Type=None,Name="None"),(Type=None,Name="None"),(Type=None,Name="None"),(Type=None,Name="None"))
    Space=GlobalSpace
    bInitial=False
    Items.3=@n1.End Control
    Items.1=@n1.Mid Control
    Items.2=@n1.Mid Control
    Items.0=@n1.Start Control
   )
   RigVMFunction_MathTransformArrayToSRT = &RigVMFunction_MathTransformArrayToSRT(Transforms=@n2.Transforms)
   > RigVMFunctionReturnNode(Spline=/Script/ControlRigSpline.RigUnit_ControlRigSplineFromPoints(
     Compression=0
     SamplesPerSegment=16
     SplineMode=BSpline
     Stretch=0
     bClosed=False
     Points=@RigVMFunction_MathTransformArrayToSRT.Translations
    ) {is=struct RigVMUnitNode}.Spline)
  ]}
 New Function_ContainedGraph {nodes [
   RigVMFunctionEntryNode()
   n1 = RigVMVariableNode() {variable=Mid Control}
   n2 = /Script/ControlRig.RigUnit_GetFloatAnimationChannel(Channel=Pelvis Chest, bInitial=False, Control=@n1.Value.Name) {is=struct RigVMUnitNode, template="GetAnimationChannel::Execute(out Value,in Control,in Channel,in bInitial)"}
   Remap = &RigVMFunction_MathDoubleRemap(SourceMaximum=1, SourceMinimum=0.5, TargetMaximum=0, TargetMinimum=1, Value=@n2.Value)
   n3 = &RigUnit_GetFloatAnimationChannel(Channel=In Between, bInitial=False, Control=@n1.Value.Name) {template="GetAnimationChannel::Execute(out Value,in Control,in Channel,in bInitial)"}
   > &RigUnit_ParentConstraint(
    AdvancedSettings=(InterpolationType=Average,RotationOrderForFilter=XZY)
    Parents=((Item=(Type=Bone,Name="None"),Weight=1),(Item=(Type=Bone,Name="None"),Weight=1))
    Parents.0.Weight=@Remap.Result
    Parents.1.Weight=&RigVMFunction_MathDoubleRemap(SourceMaximum=0.5, SourceMinimum=0, TargetMaximum=1, TargetMinimum=0, Value=@n2.Value).Result
    Weight=@n3.Value
    Parents.0.Item=RigVMVariableNode() {variable=Start Driver}.Value
    Parents.1.Item=RigVMVariableNode() {variable=End Driver}.Value
    Child=RigVMVariableNode() {variable=Mid Null}.Value
   )
   n4 = RigVMVariableNode() {variable=Mid Null}
   n5 = &RigUnit_HierarchyGetParent(bDefaultParent=True, Child=@n4.Value)
   Subtract = &RigVMFunction_MathDoubleSub(A=1, B=@n3.Value)
   > &RigUnit_SetTransform(
    bInitial=False
    Value=/Script/RigVM.RigVMFunction_MathTransformMakeAbsolute(
     Local=&RigUnit_GetTransform(Space=LocalSpace, bInitial=true, Item=@n4.Value).Transform
     Parent=&RigUnit_GetTransform(bInitial=False, Item=@n5.Parent).Transform
    ) {is=struct RigVMUnitNode, template="Make Absolute::Execute(in Local,in Parent,out Global)"}.Global
    Item=@n4.Value
    Weight=@Subtract.Result
   )
   > RigVMFunctionReturnNode()
   RigVMVariableNode() {variable=Start Control}
   RigVMVariableNode() {variable=End Control}
  ]}
 New Function_ContainedGraph {nodes [
   n1 = RigVMFunctionEntryNode()
   Get Control Color From Metadata_1_1 = RigVMFunctionReferenceNode(Color Override=RigVMVariableNode() {variable=Color}.Value)
   RigVMFunction_MathBoolAnd = &RigVMFunction_MathBoolAnd(
    A=RigVMDispatchNode(B=@n1.Color Override.G, A=@n1.Color Override.R) {template="DISPATCH_RigVMDispatch_CoreEquals(in A,in B,out Result)"}.Result
    B=RigVMDispatchNode(B=@n1.Color Override.B, A=@n1.Color Override.G) {template="DISPATCH_RigVMDispatch_CoreEquals(in A,in B,out Result)"}.Result
   )
   > RigVMFunctionReturnNode(Result=RigVMDispatchNode(False=@n1.Color Override, True=@"Get Control Color From Metadata_1_1".Result, Condition=@RigVMFunction_MathBoolAnd.Result) {template="DISPATCH_RigVMDispatch_If(in Condition,in True,in False,out Result)"}.Result)
  ]}
 New Function_ContainedGraph {nodes [
   RigVMFunctionEntryNode()
   > RigVMFunction_Sequence = RigVMAggregateNode()
   For_Each_2 = RigVMDispatchNode(ExecuteContext=@RigVMFunction_Sequence.B, Array=RigVMVariableNode() {variable=IK Gimbal Controls}.Value) {template="DISPATCH_RigVMDispatch_ArrayIterator(in Array,out Element,out Index,out Count,out Ratio)"}
   n1 = /Script/ControlRig.RigUnit_GetBoolAnimationChannelFromItem(bInitial=False, Item=RigVMVariableNode() {variable=IKs Visibility Channel}.Value) {is=struct RigVMUnitNode, template="GetAnimationChannelFromItem::Execute(out Value,in Item,in bInitial)"}
   From_String_1 = RigVMDispatchNode(String=Gimbal Control Vis) {template="DISPATCH_RigDispatch_FromString(in String,out Result)"}
   > /Script/ControlRig.RigUnit_SetControlVisibility(
    Item=@For_Each_2.Element
    bVisible=&RigVMFunction_MathBoolAnd(
     A=@n1.Value
     B=/Script/ControlRig.RigUnit_GetBoolAnimationChannel(bInitial=False, Control=@For_Each_2.Element.Name, Channel=@From_String_1.Result) {is=struct RigVMUnitNode, template="GetAnimationChannel::Execute(out Value,in Control,in Channel,in bInitial)"}.Value
    ).Result
   ) {is=struct RigVMUnitNode, template="SetControlVisibility::Execute(in Item,in Pattern,in bVisible)"}
   For_Each_4 = RigVMDispatchNode(ExecuteContext=@RigVMFunction_Sequence.C, Array=RigVMVariableNode() {variable=FK Controls}.Value) {template="DISPATCH_RigVMDispatch_ArrayIterator(in Array,out Element,out Index,out Count,out Ratio)"}
   n2 = &RigUnit_GetBoolAnimationChannelFromItem(bInitial=False, Item=RigVMVariableNode() {variable=FKs Visibility Channel}.Value) {template="GetAnimationChannelFromItem::Execute(out Value,in Item,in bInitial)"}
   > &RigUnit_SetControlVisibility(Item=@For_Each_4.Element, bVisible=@n2.Value) {template="SetControlVisibility::Execute(in Item,in Pattern,in bVisible)"}
   For_Each = RigVMDispatchNode(ExecuteContext=@RigVMFunction_Sequence.D, Array=RigVMVariableNode() {variable=FK Gimbal Controls}.Value) {template="DISPATCH_RigVMDispatch_ArrayIterator(in Array,out Element,out Index,out Count,out Ratio)"}
   From_String = RigVMDispatchNode(String=Gimbal Control Vis) {template="DISPATCH_RigDispatch_FromString(in String,out Result)"}
   GetAnimationChannel = &RigUnit_GetBoolAnimationChannel(bInitial=False, Control=@For_Each.Element.Name, Channel=@From_String.Result) {template="GetAnimationChannel::Execute(out Value,in Control,in Channel,in bInitial)"}
   RigVMFunction_MathBoolAnd = &RigVMFunction_MathBoolAnd(B=@GetAnimationChannel.Value, A=@n2.Value)
   > &RigUnit_SetControlVisibility(Item=@For_Each.Element, bVisible=@RigVMFunction_MathBoolAnd.Result) {template="SetControlVisibility::Execute(in Item,in Pattern,in bVisible)"}
   For_Each_1 = RigVMDispatchNode(ExecuteContext=@RigVMFunction_Sequence.E, Array=RigVMVariableNode() {variable=Sec FK Controls}.Value) {template="DISPATCH_RigVMDispatch_ArrayIterator(in Array,out Element,out Index,out Count,out Ratio)"}
   > &RigUnit_SetControlVisibility(
    Item=@For_Each_1.Element
    bVisible=&RigUnit_GetBoolAnimationChannelFromItem(bInitial=False, Item=RigVMVariableNode() {variable=Sec FKs Visibility Channel}.Value) {template="GetAnimationChannelFromItem::Execute(out Value,in Item,in bInitial)"}.Value
   ) {template="SetControlVisibility::Execute(in Item,in Pattern,in bVisible)"}
   RigVMFunctionReturnNode(ExecuteContext=@RigVMFunction_Sequence.F)
   n3 = RigVMVariableNode() {variable=Mid Control}
   n4 = RigVMVariableNode() {variable=End Control}
   Make_Array = RigVMDispatchNode(Values=((Type=None,Name="None"),((),Type=None,Name="None")), Values.0=@n3.Value, Values.1=@n4.Value) {template="DISPATCH_RigVMDispatch_ArrayMake(in Values,out Array)"}
   Make_Array_1 = RigVMDispatchNode(
    Values=((Type=None,Name="None"),((),Type=None,Name="None"),((),Type=None,Name="None"))
    Values.0=RigVMVariableNode() {variable=Start Control}.Value
    Values.1=@n3.Value
    Values.2=@n4.Value
   ) {template="DISPATCH_RigVMDispatch_ArrayMake(in Values,out Array)"}
   For_Each_3 = RigVMDispatchNode(
    Array=RigVMDispatchNode(False=@Make_Array.Array, True=@Make_Array_1.Array, Condition=RigVMVariableNode() {variable=Is Neck}.Value) {template="DISPATCH_RigVMDispatch_If(in Condition,in True,in False,out Result)"}.Result
    ExecuteContext=@RigVMFunction_Sequence.A
   ) {template="DISPATCH_RigVMDispatch_ArrayIterator(in Array,out Element,out Index,out Count,out Ratio)"}
   > &RigUnit_SetControlVisibility(Item=@For_Each_3.Element, bVisible=@n1.Value) {template="SetControlVisibility::Execute(in Item,in Pattern,in bVisible)"}
  ]}
 New Function_ContainedGraph {nodes [
   n1 = RigVMFunctionEntryNode()
   n2 = RigVMDispatchNode(Array=@n1.Array) {template="DISPATCH_RigVMDispatch_ArrayGetNum(in Array,out Num)"}
   GreaterEqual = /Script/RigVM.RigVMFunction_MathIntGreaterEqual(B=1, A=@n2.Num) {is=struct RigVMUnitNode, template="GreaterEqual::Execute(in A,in B,out Result)"}
   Less = /Script/RigVM.RigVMFunction_MathIntLess(A=@n1.Index, B=@n2.Num) {is=struct RigVMUnitNode, template="Less::Execute(in A,in B,out Result)"}
   RigVMFunction_MathBoolAnd = &RigVMFunction_MathBoolAnd(B=@GreaterEqual.Result, A=@Less.Result)
   > RigVMFunctionReturnNode(Display Name=RigVMDispatchNode(
     False=None
     True=RigVMDispatchNode(Array=@n1.Array, Index=@n1.Index) {template="DISPATCH_RigVMDispatch_ArrayGetAtIndex(in Array,in Index,out Element)"}.Element
     Condition=@RigVMFunction_MathBoolAnd.Result
    ) {template="DISPATCH_RigVMDispatch_If(in Condition,in True,in False,out Result)"}.Result)
  ]}
 RigVMFunctionLibrary {nodes [
   Build Spline From Items = RigVMCollapseNode(End Control=(Type=None,Name="None"), Mid Control=(Type=None,Name="None"), Start Control=(Type=None,Name="None"))
   Construct Sliding Proxy = RigVMCollapseNode(
    Driven Control=(Type=Control,Name="body_ctrl")
    ExecuteContext=()
    Proxy Name=proxy
    Proxy Parent Control=(Type=None,Name="leg_l_controls_space")
    Proxy Shape Settings="(DisplayName=\"None\",bSelectable=True,InitialSpace=LocalSpace,bUsePreferredRotationOrder=False,PreferredRotationOrder=YZX,Limits=(LimitTranslationX=(bMinimum=False,bMaximum=False),LimitTranslationY=(bMinimum=False,bMaximum=False),LimitTranslationZ=(bMinimum ...(truncated)"
    Proxy Snap To=(Type=None,Name="None")
   )
   Control Color Override From Metadata = RigVMCollapseNode(Color Override=(R=1,G=1,B=1,A=1))
   Controls Visibility = RigVMCollapseNode(ExecuteContext=())
   Forward Sliding Proxy = RigVMCollapseNode(
    Driven Control=(Type=Control,Name="body_ctrl")
    ExecuteContext=()
    Proxy Buffer=(Type=Null,Name="foot_l_proxy_bfr")
    Proxy Control=(Type=Control,Name="foot_l_proxy")
    Proxy Slide Channel=(Type=Control,Name="body_ctrl")
    Proxy Vis Channel=(Type=Control,Name="body_ctrl")
   )
   Get Optional Display Name = RigVMCollapseNode(Array=())
   Solve Mid Control Null = RigVMCollapseNode(ExecuteContext=())
  ]}
 RigVMFunction_Sequence_1_1_ContainedGraph {nodes [
   n1 = RigVMFunctionEntryNode()
   n2 = &RigVMFunction_Sequence(ExecuteContext=@n1.ExecuteContext)
   n3 = &RigVMFunction_Sequence(ExecuteContext=@n2.B)
   n4 = &RigVMFunction_Sequence(ExecuteContext=@n3.B)
   n5 = &RigVMFunction_Sequence(ExecuteContext=@n4.B)
   n6 = &RigVMFunction_Sequence(ExecuteContext=@n5.B)
   n7 = &RigVMFunction_Sequence(ExecuteContext=@n6.B)
   RigVMFunctionReturnNode(A=@n2.A, B=@n3.A, C=@n4.A, D=@n5.A, E=@n6.A, F=@n7.A, G=@n7.B)
  ]}
 RigVMFunction_Sequence_1_1_ContainedGraph {nodes [
   n1 = RigVMFunctionEntryNode()
   n2 = &RigVMFunction_Sequence(ExecuteContext=@n1.ExecuteContext)
   n3 = &RigVMFunction_Sequence(ExecuteContext=@n2.B)
   n4 = &RigVMFunction_Sequence(ExecuteContext=@n3.B)
   RigVMFunctionReturnNode(A=@n2.A, B=@n3.A, C=@n4.A, D=@n4.B)
  ]}
 RigVMFunction_Sequence_1_ContainedGraph {nodes [
   n1 = RigVMFunctionEntryNode()
   RigVMFunction_Sequence = &RigVMFunction_Sequence(ExecuteContext=@n1.ExecuteContext)
   n2 = &RigVMFunction_Sequence(ExecuteContext=@RigVMFunction_Sequence.B)
   n3 = &RigVMFunction_Sequence(ExecuteContext=@n2.B)
   n4 = &RigVMFunction_Sequence(ExecuteContext=@n3.B)
   n5 = &RigVMFunction_Sequence(ExecuteContext=@n4.B)
   n6 = &RigVMFunction_Sequence(ExecuteContext=@n5.B)
   n7 = &RigVMFunction_Sequence(ExecuteContext=@n6.B)
   n8 = &RigVMFunction_Sequence(ExecuteContext=@n7.B)
   RigVMFunctionReturnNode(A=@RigVMFunction_Sequence.A, B=@n2.A, C=@n3.A, D=@n4.A, E=@n5.A, F=@n6.A, G=@n7.A, H=@n8.A, I=@n8.B)
  ]}
 RigVMFunction_Sequence_1_ContainedGraph {nodes [
   n1 = RigVMFunctionEntryNode()
   RigVMFunction_Sequence = &RigVMFunction_Sequence(ExecuteContext=@n1.ExecuteContext)
   n2 = &RigVMFunction_Sequence(ExecuteContext=@RigVMFunction_Sequence.B)
   n3 = &RigVMFunction_Sequence(ExecuteContext=@n2.B)
   n4 = &RigVMFunction_Sequence(ExecuteContext=@n3.B)
   n5 = &RigVMFunction_Sequence(ExecuteContext=@n4.B)
   RigVMFunctionReturnNode(A=@RigVMFunction_Sequence.A, B=@n2.A, C=@n3.A, D=@n4.A, E=@n5.A, F=@n5.B)
  ]}
 RigVMFunction_Sequence_2_ContainedGraph {nodes [
   n1 = RigVMFunctionEntryNode()
   n2 = &RigVMFunction_Sequence(ExecuteContext=@n1.ExecuteContext)
   n3 = &RigVMFunction_Sequence(ExecuteContext=@n2.B)
   RigVMFunctionReturnNode(A=@n2.A, B=@n3.A, C=@n3.B)
  ]}
 RigVMFunction_Sequence_3_ContainedGraph {nodes [
   n1 = RigVMFunctionEntryNode()
   RigVMFunction_Sequence = &RigVMFunction_Sequence(ExecuteContext=@n1.ExecuteContext)
   n2 = &RigVMFunction_Sequence(ExecuteContext=@RigVMFunction_Sequence.B)
   n3 = &RigVMFunction_Sequence(ExecuteContext=@n2.B)
   n4 = &RigVMFunction_Sequence(ExecuteContext=@n3.B)
   n5 = &RigVMFunction_Sequence(ExecuteContext=@n4.B)
   RigVMFunctionReturnNode(A=@RigVMFunction_Sequence.A, B=@n2.A, C=@n3.A, D=@n4.A, E=@n5.A, F=@n5.B)
  ]}
 RigVMModel {nodes [
   RigUnit_BeginExecution = &RigUnit_BeginExecution()
   n1 = &RigVMFunction_Sequence(ExecuteContext=@RigUnit_BeginExecution.ExecutePin)
   Controls Visibility = RigVMFunctionReferenceNode(ExecuteContext=@n1.A)
   > RigVMFunction_Sequence_1_1 = RigVMAggregateNode()
   Solve Mid Control Null = RigVMFunctionReferenceNode(ExecuteContext=@RigVMFunction_Sequence_1_1.D)
   n2 = RigVMVariableNode() {variable=Chest Null}
   n3 = RigVMVariableNode() {variable=End Control}
   RigUnit_GetFloatAnimationChannel = &RigUnit_GetFloatAnimationChannel(Channel=Distribute Rotation, bInitial=False, Control=@n3.Value.Name) {template="GetAnimationChannel::Execute(out Value,in Control,in Channel,in bInitial)"}
   RigVMFunction_MathQuaternionToEuler = &RigVMFunction_MathQuaternionToEuler(RotationOrder=ZYX, Value=&RigUnit_GetTransform(Space=LocalSpace, bInitial=False, Item=@n3.Value).Transform.Rotation)
   Scale = &RigVMFunction_MathVectorScale(Factor=@RigUnit_GetFloatAnimationChannel.Value, Value=@RigVMFunction_MathQuaternionToEuler.Result) {template="Scale::Execute(in Value,in Factor,out Result)"}
   RigVMFunction_MathQuaternionFromEuler = &RigVMFunction_MathQuaternionFromEuler(Euler=(X=0,Y=0,Z=0), Euler.X=@Scale.Result.X, Euler.Y=@Scale.Result.Y)
   Set Transform_2 = &RigUnit_SetRotation(
    Weight=1
    bInitial=False
    Value=/Script/RigVM.RigVMFunction_MathQuaternionMakeAbsolute(
     Parent=&RigUnit_GetTransform(Space=LocalSpace, bInitial=true, Item=@n2.Value).Transform.Rotation
     Local=@RigVMFunction_MathQuaternionFromEuler.Result
    ) {is=struct RigVMUnitNode, template="Make Absolute::Execute(in Local,in Parent,out Global)"}.Global
    Item=@n2.Value
    ExecutePin=@RigVMFunction_Sequence_1_1.C
   )
   Evaluate Space Nulls v01_2 = RigVMFunctionReferenceNode(
    Orient Only=true
    Space Nulls=RigVMVariableNode() {variable=End IK Space Nulls}.Value
    Spaces=RigVMVariableNode() {variable=End Spaces}.Value
    ExecuteContext=@RigVMFunction_Sequence_1_1.B
   )
   > Evaluate Space Nulls v01_3 = RigVMFunctionReferenceNode(
    Orient Only=true
    Spaces=RigVMVariableNode() {variable=End Fk Spaces}.Value
    Space Nulls=RigVMVariableNode() {variable=End FK Space Nulls}.Value
   )
   Evaluate Space Nulls v01 = RigVMFunctionReferenceNode(
    Orient Only=true
    Spaces=RigVMVariableNode() {variable=Start Spaces}.Value
    Space Nulls=RigVMVariableNode() {variable=Start IK Space Nulls}.Value
    ExecuteContext=@RigVMFunction_Sequence_1_1.A
   )
   > Evaluate Space Nulls v01_1 = RigVMFunctionReferenceNode(
    Orient Only=true
    Space Nulls=RigVMVariableNode() {variable=Start FK Space Nulls}.Value
    Spaces=RigVMVariableNode() {variable=Start Fk Spaces}.Value
   )
   RigVMFunction_Sequence_1 = RigVMAggregateNode(ExecuteContext=@n1.B)
   RigVMFunction_MathBoolNot = &RigVMFunction_MathBoolNot(Value=RigVMVariableNode() {variable=Is Neck}.Value)
   RigVMFunction_ControlFlowBranch = &RigVMFunction_ControlFlowBranch(Condition=@RigVMFunction_MathBoolNot.Result, ExecuteContext=@RigVMFunction_Sequence_1.E)
   n4 = RigVMVariableNode() {variable=End Control}
   Forward Movable Proxy v01 = RigVMFunctionReferenceNode(
    ExecuteContext=@RigVMFunction_ControlFlowBranch.True
    Proxy Control=RigVMVariableNode() {variable=End Proxy Control}.Value
    Driven Control=@n4.Value
    Proxy Snap To=@n4.Value
    Proxy Buffer=RigVMVariableNode() {variable=End Proxy Buffer}.Value
    Proxy Vis Channel=RigVMVariableNode() {variable=End Proxy Vis Channel}.Value
   )
   Build Spline From Items_1 = RigVMFunctionReferenceNode(
    Start Control=RigVMVariableNode() {variable=Start Driver}.Value
    End Control=RigVMVariableNode() {variable=End Driver}.Value
    Mid Control=RigVMVariableNode() {variable=Mid Control}.Value
   )
   n5 = RigVMVariableNode(Value=@"Build Spline From Items_1".Spline, ExecuteContext=@RigVMFunction_Sequence_1.A) {variable=Spline}
   n6 = &RigVMFunction_ControlFlowBranch(Condition=RigVMVariableNode() {variable=Debug}.Value, ExecuteContext=@n5.ExecuteContext)
   /Script/ControlRigSpline.RigUnit_DrawControlRigSpline(Color=(R=0.5,G=0,B=0,A=1), Detail=16, Thickness=0.1, Spline=@"Build Spline From Items_1".Spline, ExecutePin=@n6.True) {is=struct RigVMUnitNode}
   n7 = RigVMDispatchNode(Index=0, Array=RigVMVariableNode() {variable=Bones}.Value) {template="DISPATCH_RigVMDispatch_ArrayGetAtIndex(in Array,in Index,out Element)"}
   n8 = RigVMVariableNode() {variable=Start Control}
   &RigUnit_SetTransform(
    Weight=1
    bInitial=False
    Item=@n7.Element
    Value=/Script/ControlRig.RigUnit_ProjectTransformToNewParent(bChildInitial=True, bNewParentInitial=False, bOldParentInitial=True, Child=@n7.Element, NewParent=@n8.Value, OldParent=@n8.Value) {is=struct RigVMUnitNode}.Transform
    ExecutePin=@n6.Completed
   )
   n9 = RigVMDispatchNode(Array=RigVMVariableNode() {variable=Bones}.Value, ExecuteContext=@RigVMFunction_Sequence_1.D) {template="DISPATCH_RigVMDispatch_ArrayIterator(in Array,out Element,out Index,out Count,out Ratio)"}
   > /Script/ControlRig.RigUnit_SetControlOffset(
    Space=LocalSpace
    Offset=&RigUnit_GetTransform(Space=LocalSpace, bInitial=False, Item=@n9.Element).Transform
    Control=RigVMDispatchNode(Index=@n9.Index, Array=RigVMVariableNode() {variable=Sec FK Controls}.Value) {template="DISPATCH_RigVMDispatch_ArrayGetAtIndex(in Array,in Index,out Element)"}.Element.Name
   ) {is=struct RigVMUnitNode, template="SetControlOffset::Execute(in Control,io Offset,in Space)"}
   n10 = RigVMDispatchNode(ExecuteContext=@n9.Completed, Array=RigVMVariableNode() {variable=Bones}.Value) {template="DISPATCH_RigVMDispatch_ArrayIterator(in Array,out Element,out Index,out Count,out Ratio)"}
   n11 = RigVMVariableNode() {variable=Sec FK Controls}
   > &RigUnit_SetTransform(
    Weight=1
    bInitial=False
    Item=@n10.Element
    Value=&RigUnit_GetTransform(
     bInitial=False
     Item=RigVMDispatchNode(Index=@n10.Index, Array=@n11.Value) {template="DISPATCH_RigVMDispatch_ArrayGetAtIndex(in Array,in Index,out Element)"}.Element
    ).Transform
   )
   n12 = RigVMDispatchNode(Index=-1, Array=RigVMVariableNode() {variable=Bones}.Value) {template="DISPATCH_RigVMDispatch_ArrayGetAtIndex(in Array,in Index,out Element)"}
   n13 = RigVMVariableNode() {variable=End Driver}
   Set Transform_1_1 = &RigUnit_SetRotation(
    Space=GlobalSpace
    Weight=1
    bInitial=False
    Item=@n12.Element
    Value=&RigUnit_ProjectTransformToNewParent(bChildInitial=True, bNewParentInitial=False, bOldParentInitial=True, Child=@n12.Element, NewParent=@n13.Value, OldParent=@n13.Value).Transform.Rotation
    ExecutePin=@RigVMFunction_Sequence_1.C
   )
   Clone = RigVMDispatchNode(Array=RigVMVariableNode() {variable=Bones}.Value) {template="DISPATCH_RigVMDispatch_ArrayClone(in Array,out Clone)"}
   Remove = RigVMDispatchNode(Index=0, Array=@Clone.Clone, ExecuteContext=@RigVMFunction_Sequence_1.B) {template="DISPATCH_RigVMDispatch_ArrayRemove(io Array,in Index)"}
   > DISPATCH_RigVMDispatch_ArrayIterator = RigVMDispatchNode(Array=@Remove.Array) {template="DISPATCH_RigVMDispatch_ArrayIterator(in Array,out Element,out Index,out Count,out Ratio)"}
   GetAnimationChannelFromItem = &RigUnit_GetFloatAnimationChannelFromItem(bInitial=False, Item=RigVMVariableNode() {variable=Stretch Channel}.Value) {template="GetAnimationChannelFromItem::Execute(out Value,in Item,in bInitial)"}
   n14 = RigVMDispatchNode(Default=0, Name=Bone Percentage, NameSpace=Self, Item=@DISPATCH_RigVMDispatch_ArrayIterator.Element) {template="DISPATCH_RigDispatch_GetMetadata(in Item,in Name,in NameSpace,in Default,out Value,out Found)"}
   n15 = RigVMVariableNode() {variable=Spline}
   Divide = &RigVMFunction_MathDoubleDiv(
    A=&RigUnit_GetLengthControlRigSpline(Spline=@n15.Value).Length
    B=RigVMVariableNode() {variable=Spline Orig Length}.Value
   )
   Interpolate = /Script/RigVM.RigVMFunction_MathQuaternionSlerp(
    A=&RigUnit_GetTransform(bInitial=False, Item=RigVMVariableNode() {variable=Start Driver}.Value).Transform.Rotation
    B=&RigUnit_GetTransform(bInitial=False, Item=RigVMVariableNode() {variable=End Driver}.Value).Transform.Rotation
    T=@n14.Value
   ) {is=struct RigVMUnitNode, template="Interpolate::Execute(in A,in B,in T,out Result)"}
   n16 = /Script/RigVM.RigVMFunction_MathQuaternionRotateVector(Transform=@Interpolate.Result, Vector=RigVMVariableNode() {variable=Up Vector Axis}.Value) {is=struct RigVMUnitNode, template="Rotate Vector::Execute(in Transform,in Vector,out Result)"}
   n17 = /Script/ControlRigSpline.RigUnit_TransformFromControlRigSpline(
    Roll=0
    U=/Script/ControlRigSpline.RigUnit_ParameterAtPercentage(Percentage=&RigVMFunction_MathDoubleDiv(A=@n14.Value, B=@Divide.Result).Result, Spline=@n15.Value) {is=struct RigVMUnitNode}.U
    UpVector=@n16.Result
    Spline=@n15.Value
   ) {is=struct RigVMUnitNode}
   n18 = &RigUnit_TransformFromControlRigSpline(
    Roll=0
    U=&RigUnit_ParameterAtPercentage(Spline=@n15.Value, Percentage=@n14.Value).U
    UpVector=@n16.Result
    Spline=@n15.Value
   )
   Greater = /Script/RigVM.RigVMFunction_MathDoubleGreater(B=1, A=@Divide.Result) {is=struct RigVMUnitNode, template="Greater::Execute(in A,in B,out Result)"}
   > &RigUnit_SetTransform(
    Weight=1
    bInitial=False
    Item=@DISPATCH_RigVMDispatch_ArrayIterator.Element
    Value=/Script/RigVM.RigVMFunction_MathTransformLerp(
     T=@GetAnimationChannelFromItem.Value
     A=RigVMDispatchNode(True=@n17.Transform, False=@n18.Transform, Condition=@Greater.Result) {template="DISPATCH_RigVMDispatch_If(in Condition,in True,in False,out Result)"}.Result
     B=@n18.Transform
    ) {is=struct RigVMUnitNode, template="Interpolate::Execute(in A,in B,in T,out Result)"}.Result
   )
   RigVMVariableNode() {variable=End Control}
   RigVMVariableNode() {variable=Start Control}
   RigVMVariableNode() {variable=End Control}
  ]}
 RigVMModel Backward Solve {nodes [
   n1 = /Script/ControlRig.RigUnit_InverseExecution() {is=struct RigVMUnitNode}
   RigVMFunction_Sequence = RigVMAggregateNode(ExecuteContext=@n1.ExecutePin)
   RigVMFunction_Sequence_1 = RigVMAggregateNode(ExecuteContext=@RigVMFunction_Sequence.F)
   /Script/ControlRig.RigUnit_SetBoolAnimationChannelFromItem(Value=true, bInitial=False, ExecutePin=@RigVMFunction_Sequence_1.A, Item=RigVMVariableNode() {variable=Sec FKs Visibility Channel}.Value) {is=struct RigVMUnitNode, template="SetAnimationChannelFromItem::Execute(in Value,in Item,in bInitial)"}
   > SetAnimationChannelFromItem = /Script/ControlRig.RigUnit_SetFloatAnimationChannelFromItem(Value=1, bInitial=False, Item=RigVMVariableNode() {variable=Stretch Channel}.Value) {is=struct RigVMUnitNode, template="SetAnimationChannelFromItem::Execute(in Value,in Item,in bInitial)"}
   Build Spline From Items_1 = RigVMFunctionReferenceNode(
    End Control=RigVMVariableNode() {variable=End Driver}.Value
    Start Control=RigVMVariableNode() {variable=Start Driver}.Value
    Mid Control=RigVMVariableNode() {variable=Mid Control}.Value
   )
   > RigVMVariableNode(Value=@"Build Spline From Items_1".Spline) {variable=Spline}
   > DISPATCH_RigVMDispatch_ArrayIterator = RigVMDispatchNode(Array=RigVMVariableNode() {variable=Bones}.Value) {template="DISPATCH_RigVMDispatch_ArrayIterator(in Array,out Element,out Index,out Count,out Ratio)"}
   Subtract = /Script/RigVM.RigVMFunction_MathIntSub(
    B=1
    A=RigVMDispatchNode(Array=RigVMVariableNode() {variable=Bones}.Value) {template="DISPATCH_RigVMDispatch_ArrayGetNum(in Array,out Num)"}.Num
   ) {is=struct RigVMUnitNode, template="Subtract::Execute(in A,in B,out Result)"}
   n2 = RigVMDispatchNode(B=0, A=@DISPATCH_RigVMDispatch_ArrayIterator.Index) {template="DISPATCH_RigVMDispatch_CoreEquals(in A,in B,out Result)"}
   GetAnimationChannelFromItem = &RigUnit_GetFloatAnimationChannelFromItem(bInitial=False, Item=RigVMVariableNode() {variable=Stretch Channel}.Value) {template="GetAnimationChannelFromItem::Execute(out Value,in Item,in bInitial)"}
   n3 = RigVMDispatchNode(Default=0, Name=Bone Percentage, NameSpace=Self, Item=@DISPATCH_RigVMDispatch_ArrayIterator.Element) {template="DISPATCH_RigDispatch_GetMetadata(in Item,in Name,in NameSpace,in Default,out Value,out Found)"}
   n4 = RigVMVariableNode() {variable=Spline}
   Divide = &RigVMFunction_MathDoubleDiv(
    A=&RigUnit_GetLengthControlRigSpline(Spline=@n4.Value).Length
    B=RigVMVariableNode() {variable=Spline Orig Length}.Value
   )
   n5 = &RigUnit_ParameterAtPercentage(Percentage=&RigVMFunction_MathDoubleDiv(A=@n3.Value, B=@Divide.Result).Result, Spline=@n4.Value)
   Interpolate = &RigVMFunction_MathQuaternionSlerp(
    A=&RigUnit_GetTransform(bInitial=False, Item=RigVMVariableNode() {variable=Start Driver}.Value).Transform.Rotation
    B=&RigUnit_GetTransform(bInitial=False, Item=RigVMVariableNode() {variable=End Driver}.Value).Transform.Rotation
    T=@n3.Value
   ) {template="Interpolate::Execute(in A,in B,in T,out Result)"}
   n6 = &RigVMFunction_MathQuaternionRotateVector(Transform=@Interpolate.Result, Vector=RigVMVariableNode() {variable=Up Vector Axis}.Value) {template="Rotate Vector::Execute(in Transform,in Vector,out Result)"}
   n7 = &RigUnit_TransformFromControlRigSpline(
    Roll=0
    U=&RigUnit_ParameterAtPercentage(Spline=@n4.Value, Percentage=@n3.Value).U
    UpVector=@n6.Result
    Spline=@n4.Value
   )
   Greater = &RigVMFunction_MathDoubleGreater(B=1, A=@Divide.Result) {template="Greater::Execute(in A,in B,out Result)"}
   n8 = &RigVMFunction_MathTransformLerp(
    T=@GetAnimationChannelFromItem.Value
    A=RigVMDispatchNode(
     True=&RigUnit_TransformFromControlRigSpline(Roll=0, U=@n5.U, UpVector=@n6.Result, Spline=@n4.Value).Transform
     False=@n7.Transform
     Condition=@Greater.Result
    ) {template="DISPATCH_RigVMDispatch_If(in Condition,in True,in False,out Result)"}.Result
    B=@n7.Transform
   ) {template="Interpolate::Execute(in A,in B,in T,out Result)"}
   n9 = RigVMVariableNode() {variable=Start Control}
   Get_Metadata_5 = RigVMDispatchNode(Default=(Rotation=(X=0,Y=0,Z=0,W=1),Translation=(X=0,Y=0,Z=0),Scale3D=(X=1,Y=1,Z=1)), Name=Offset, NameSpace=Self, Item=@n9.Value) {template="DISPATCH_RigDispatch_GetMetadata(in Item,in Name,in NameSpace,in Default,out Value,out Found)"}
   n10 = &RigVMFunction_MathTransformMul(
    B=&RigUnit_GetTransform(bInitial=False, Item=@n9.Value).Transform
    A=/Script/RigVM.RigVMFunction_MathTransformInverse(Value=@Get_Metadata_5.Value) {is=struct RigVMUnitNode, template="Inverse::Execute(in Value,out Result)"}.Result
   ) {template="Multiply::Execute(in A,in B,out Result)"}
   n11 = RigVMVariableNode() {variable=End Control}
   n12 = &RigUnit_GetTransform(bInitial=False, Item=@n11.Value)
   Get_Metadata_6 = RigVMDispatchNode(Default=(Rotation=(X=0,Y=0,Z=0,W=1),Translation=(X=0,Y=0,Z=0),Scale3D=(X=1,Y=1,Z=1)), Name=Offset, NameSpace=Self, Item=@n11.Value) {template="DISPATCH_RigDispatch_GetMetadata(in Item,in Name,in NameSpace,in Default,out Value,out Found)"}
   n13 = &RigVMFunction_MathTransformInverse(Value=@Get_Metadata_6.Value) {template="Inverse::Execute(in Value,out Result)"}
   > &RigUnit_SetTransform(
    Weight=1
    bInitial=False
    Item=RigVMDispatchNode(Array=RigVMVariableNode() {variable=Match Nulls}.Value, Index=@DISPATCH_RigVMDispatch_ArrayIterator.Index) {template="DISPATCH_RigVMDispatch_ArrayGetAtIndex(in Array,in Index,out Element)"}.Element
    Value=RigVMDispatchNode(
     Condition=RigVMDispatchNode(A=@Subtract.Result, B=@DISPATCH_RigVMDispatch_ArrayIterator.Index) {template="DISPATCH_RigVMDispatch_CoreEquals(in A,in B,out Result)"}.Result
     False=RigVMDispatchNode(Condition=@n2.Result, False=@n8.Result, True=@n10.Result) {template="DISPATCH_RigVMDispatch_If(in Condition,in True,in False,out Result)"}.Result
     True=&RigVMFunction_MathTransformMul(B=@n12.Transform, A=@n13.Result) {template="Multiply::Execute(in A,in B,out Result)"}.Result
    ) {template="DISPATCH_RigVMDispatch_If(in Condition,in True,in False,out Result)"}.Result
   )
   n14 = RigVMDispatchNode(ExecuteContext=@RigVMFunction_Sequence_1.C, Array=RigVMVariableNode() {variable=Match Nulls}.Value) {template="DISPATCH_RigVMDispatch_ArrayIterator(in Array,out Element,out Index,out Count,out Ratio)"}
   > &RigUnit_SetControlOffset(
    Space=LocalSpace
    Control=RigVMDispatchNode(Index=@n14.Index, Array=RigVMVariableNode() {variable=Sec FK Controls}.Value) {template="DISPATCH_RigVMDispatch_ArrayGetAtIndex(in Array,in Index,out Element)"}.Element.Name
    Offset=&RigUnit_GetTransform(Space=LocalSpace, bInitial=False, Item=@n14.Element).Transform
   ) {template="SetControlOffset::Execute(in Control,io Offset,in Space)"}
   n15 = RigVMDispatchNode(ExecuteContext=@n14.Completed, Array=RigVMVariableNode() {variable=Sec FK Controls}.Value) {template="DISPATCH_RigVMDispatch_ArrayIterator(in Array,out Element,out Index,out Count,out Ratio)"}
   n16 = RigVMVariableNode() {variable=Bones}
   > &RigUnit_SetTransform(
    Weight=1
    bInitial=False
    Item=@n15.Element
    Value=&RigUnit_GetTransform(
     bInitial=False
     Item=RigVMDispatchNode(Index=@n15.Index, Array=@n16.Value) {template="DISPATCH_RigVMDispatch_ArrayGetAtIndex(in Array,in Index,out Element)"}.Element
    ).Transform
   )
   Evaluate Space Nulls v01 = RigVMFunctionReferenceNode(
    Orient Only=true
    Spaces=RigVMVariableNode() {variable=Start Spaces}.Value
    Space Nulls=RigVMVariableNode() {variable=Start IK Space Nulls}.Value
    ExecuteContext=@RigVMFunction_Sequence.C
   )
   > Evaluate Space Nulls v01_2 = RigVMFunctionReferenceNode(
    Orient Only=true
    Spaces=RigVMVariableNode() {variable=End Spaces}.Value
    Space Nulls=RigVMVariableNode() {variable=End IK Space Nulls}.Value
   )
   Evaluate Space Nulls v01_1 = RigVMFunctionReferenceNode(
    Orient Only=true
    Space Nulls=RigVMVariableNode() {variable=Start FK Space Nulls}.Value
    Spaces=RigVMVariableNode() {variable=Start Fk Spaces}.Value
    ExecuteContext=@RigVMFunction_Sequence.A
   )
   n17 = RigVMVariableNode() {variable=Bones}
   n18 = RigVMVariableNode() {variable=Start Control}
   Get_Metadata = RigVMDispatchNode(Default=(Rotation=(X=0,Y=0,Z=0,W=1),Translation=(X=0,Y=0,Z=0),Scale3D=(X=1,Y=1,Z=1)), Name=Offset, NameSpace=Self, Item=@n18.Value) {template="DISPATCH_RigDispatch_GetMetadata(in Item,in Name,in NameSpace,in Default,out Value,out Found)"}
   Multiply = &RigVMFunction_MathTransformMul(
    B=&RigUnit_GetTransform(
     bInitial=False
     Item=RigVMDispatchNode(Index=0, Array=@n17.Value) {template="DISPATCH_RigVMDispatch_ArrayGetAtIndex(in Array,in Index,out Element)"}.Element
    ).Transform
    A=@Get_Metadata.Value
   ) {template="Multiply::Execute(in A,in B,out Result)"}
   &RigUnit_SetTransform(
    Value=(Rotation=(X=0,Y=0,Z=0,W=1),Translation=(X=0,Y=0,Z=0),Scale3D=(X=1,Y=1,Z=1))
    Weight=1
    bInitial=False
    Value.Rotation=@Multiply.Result.Rotation
    Value.Translation=@Multiply.Result.Translation
    Item=@n18.Value
    ExecutePin=@RigVMFunction_Sequence.D
   )
   RigVMFunction_ControlFlowBranch = &RigVMFunction_ControlFlowBranch(
    Condition=&RigUnit_GetBoolAnimationChannelFromItem(bInitial=False, Item=RigVMVariableNode() {variable=FKs Visibility Channel}.Value) {template="GetAnimationChannelFromItem::Execute(out Value,in Item,in bInitial)"}.Value
    ExecuteContext=@RigVMFunction_Sequence.B
   )
   For_Each_1 = RigVMDispatchNode(ExecuteContext=@RigVMFunction_ControlFlowBranch.False, Array=RigVMVariableNode() {variable=FK Controls}.Value) {template="DISPATCH_RigVMDispatch_ArrayIterator(in Array,out Element,out Index,out Count,out Ratio)"}
   > &RigUnit_SetTransform(
    Space=LocalSpace
    Value=(Rotation=(X=0,Y=0,Z=0,W=1),Translation=(X=0,Y=0,Z=0),Scale3D=(X=1,Y=1,Z=1))
    Weight=1
    bInitial=False
    bPropagateToChildren=true
    Item=@For_Each_1.Element
   )
   For_Each = RigVMDispatchNode(ExecuteContext=@RigVMFunction_ControlFlowBranch.True, Array=RigVMVariableNode() {variable=FK Controls}.Value) {template="DISPATCH_RigVMDispatch_ArrayIterator(in Array,out Element,out Index,out Count,out Ratio)"}
   n19 = &RigVMFunction_ControlFlowBranch(
    Condition=RigVMDispatchNode(B=1, A=@For_Each.Ratio) {template="DISPATCH_RigVMDispatch_CoreEquals(in A,in B,out Result)"}.Result
    ExecuteContext=@For_Each.ExecuteContext
   )
   n20 = RigVMVariableNode() {variable=FK Bones}
   Get_Metadata_3 = RigVMDispatchNode(
    Default=(Rotation=(X=0,Y=0,Z=0,W=1),Translation=(X=0,Y=0,Z=0),Scale3D=(X=1,Y=1,Z=1))
    Name=FkDeltaTransform
    NameSpace=Self
    Item=@For_Each.Element
   ) {template="DISPATCH_RigDispatch_GetMetadata(in Item,in Name,in NameSpace,in Default,out Value,out Found)"}
   n21 = &RigVMFunction_MathTransformMul(
    B=&RigUnit_GetTransform(
     bInitial=False
     Item=RigVMDispatchNode(Index=@For_Each.Index, Array=@n20.Value) {template="DISPATCH_RigVMDispatch_ArrayGetAtIndex(in Array,in Index,out Element)"}.Element
    ).Transform
    A=@Get_Metadata_3.Value
   ) {template="Multiply::Execute(in A,in B,out Result)"}
   &RigUnit_SetTransform(
    Value=(Rotation=(X=0,Y=0,Z=0,W=1),Translation=(X=0,Y=0,Z=0),Scale3D=(X=1,Y=1,Z=1))
    Weight=1
    bInitial=False
    bPropagateToChildren=true
    Item=@For_Each.Element
    Value.Rotation=@n21.Result.Rotation
    Value.Translation=@n21.Result.Translation
    ExecutePin=@n19.Completed
   )
   Evaluate Space Nulls v01_3 = RigVMFunctionReferenceNode(
    Orient Only=true
    Space Nulls=RigVMVariableNode() {variable=End FK Space Nulls}.Value
    Spaces=RigVMVariableNode() {variable=End Fk Spaces}.Value
    ExecuteContext=@n19.True
   )
   n22 = &RigVMFunction_ControlFlowBranch(
    Condition=&RigUnit_GetBoolAnimationChannelFromItem(bInitial=False, Item=RigVMVariableNode() {variable=IKs Visibility Channel}.Value) {template="GetAnimationChannelFromItem::Execute(out Value,in Item,in bInitial)"}.Value
    ExecuteContext=@RigVMFunction_Sequence.E
   )
   &RigUnit_SetTransform(
    Space=LocalSpace
    Value=(Rotation=(X=0,Y=0,Z=0,W=1),Translation=(X=0,Y=0,Z=0),Scale3D=(X=1,Y=1,Z=1))
    Weight=1
    bInitial=False
    bPropagateToChildren=true
    ExecutePin=@n22.False
    Item=RigVMVariableNode() {variable=End Control}.Value
   )
   > &RigUnit_SetTransform(
    Space=LocalSpace
    Value=(Rotation=(X=0,Y=0,Z=0,W=1),Translation=(X=0,Y=0,Z=0),Scale3D=(X=1,Y=1,Z=1))
    Weight=1
    bInitial=False
    bPropagateToChildren=true
    Item=RigVMVariableNode() {variable=Mid Control}.Value
   )
   > Solve Mid Control Null_1 = RigVMFunctionReferenceNode()
   n23 = RigVMVariableNode() {variable=End Control}
   SetAnimationChannel = /Script/ControlRig.RigUnit_SetFloatAnimationChannel(Channel=Distribute Rotation, Value=0, bInitial=False, ExecutePin=@n22.True, Control=@n23.Value.Name) {is=struct RigVMUnitNode, template="SetAnimationChannel::Execute(in Value,in Control,in Channel,in bInitial)"}
   n24 = RigVMVariableNode() {variable=Bones}
   Get_Metadata_2 = RigVMDispatchNode(Default=(Rotation=(X=0,Y=0,Z=0,W=1),Translation=(X=0,Y=0,Z=0),Scale3D=(X=1,Y=1,Z=1)), Name=Offset, NameSpace=Self, Item=@n23.Value) {template="DISPATCH_RigDispatch_GetMetadata(in Item,in Name,in NameSpace,in Default,out Value,out Found)"}
   n25 = &RigVMFunction_MathTransformMul(
    B=&RigUnit_GetTransform(
     bInitial=False
     Item=RigVMDispatchNode(Index=-1, Array=@n24.Value) {template="DISPATCH_RigVMDispatch_ArrayGetAtIndex(in Array,in Index,out Element)"}.Element
    ).Transform
    A=@Get_Metadata_2.Value
   ) {template="Multiply::Execute(in A,in B,out Result)"}
   > &RigUnit_SetTransform(
    Value=(Rotation=(X=0,Y=0,Z=0,W=1),Translation=(X=0,Y=0,Z=0),Scale3D=(X=1,Y=1,Z=1))
    Weight=1
    bInitial=False
    Value.Rotation=@n25.Result.Rotation
    Value.Translation=@n25.Result.Translation
    Item=@n23.Value
   )
   > Solve Mid Control Null = RigVMFunctionReferenceNode()
   n26 = RigVMVariableNode() {variable=Bones}
   Num = RigVMDispatchNode(Array=@n26.Value) {template="DISPATCH_RigVMDispatch_ArrayGetNum(in Array,out Num)"}
   n27 = &RigVMFunction_MathIntDiv(B=2, A=@Num.Num) {template="Divide::Execute(in A,in B,out Result)"}
   Get_Metadata_1 = RigVMDispatchNode(
    Default=(Rotation=(X=0,Y=0,Z=0,W=1),Translation=(X=0,Y=0,Z=0),Scale3D=(X=1,Y=1,Z=1))
    Name=Offset
    NameSpace=Self
    Item=RigVMVariableNode() {variable=Mid Control}.Value
   ) {template="DISPATCH_RigDispatch_GetMetadata(in Item,in Name,in NameSpace,in Default,out Value,out Found)"}
   n28 = &RigVMFunction_MathTransformMul(
    B=&RigUnit_GetTransform(
     bInitial=False
     Item=RigVMDispatchNode(Index=@n27.Result, Array=@n26.Value) {template="DISPATCH_RigVMDispatch_ArrayGetAtIndex(in Array,in Index,out Element)"}.Element
    ).Transform
    A=@Get_Metadata_1.Value
   ) {template="Multiply::Execute(in A,in B,out Result)"}
   n29 = &RigUnit_GetTransform(
    bInitial=False
    Item=RigVMDispatchNode(Index=1, Array=RigVMVariableNode() {variable=Bones}.Value) {template="DISPATCH_RigVMDispatch_ArrayGetAtIndex(in Array,in Index,out Element)"}.Element
   )
   n30 = &RigUnit_GetTransform(
    bInitial=False
    Item=RigVMDispatchNode(Index=-2, Array=RigVMVariableNode() {variable=Bones}.Value) {template="DISPATCH_RigVMDispatch_ArrayGetAtIndex(in Array,in Index,out Element)"}.Element
   )
   n31 = /Script/RigVM.RigVMFunction_MathRayIntersectRay(
    A=(Origin=(X=0,Y=0,Z=0),Direction=(X=0,Y=0,Z=1))
    B=(Origin=(X=0,Y=0,Z=0),Direction=(X=0,Y=0,Z=1))
    B.Origin=@n29.Transform.Translation
    A.Origin=@n30.Transform.Translation
    B.Direction=&RigVMFunction_MathQuaternionRotateVector(Vector=(X=1,Y=0,Z=0), Transform=@n29.Transform.Rotation) {template="Rotate Vector::Execute(in Transform,in Vector,out Result)"}.Result
    A.Direction=&RigVMFunction_MathQuaternionRotateVector(Vector=(X=-1,Y=0,Z=0), Transform=@n30.Transform.Rotation) {template="Rotate Vector::Execute(in Transform,in Vector,out Result)"}.Result
   ) {is=struct RigVMUnitNode}
   n32 = &RigVMFunction_MathDoubleGreater(B=0, A=@n31.RatioA) {template="Greater::Execute(in A,in B,out Result)"}
   n33 = /Script/RigVM.RigVMFunction_MathDoubleLess(B=20, A=@n31.RatioA) {is=struct RigVMUnitNode, template="Less::Execute(in A,in B,out Result)"}
   > &RigUnit_SetTransform(
    Value=(Rotation=(X=0,Y=0,Z=0,W=1),Translation=(X=0,Y=0,Z=0),Scale3D=(X=1,Y=1,Z=1))
    Weight=1
    bInitial=False
    Value.Translation=RigVMDispatchNode(False=@n28.Result.Translation, Condition=&RigVMFunction_MathBoolAnd(A=@n32.Result, B=@n33.Result).Result, True=@n31.Result) {template="DISPATCH_RigVMDispatch_If(in Condition,in True,in False,out Result)"}.Result
    Value.Rotation=@n28.Result.Rotation
    Item=RigVMVariableNode() {variable=Mid Control}.Value
   )
   RigVMVariableNode() {variable=End Control}
   RigVMVariableNode() {variable=Start Control}
  ]}
 SequenceAggregate_1_ContainedGraph {nodes [
   n1 = RigVMFunctionEntryNode()
   SequenceAggregate = &RigVMFunction_Sequence(ExecuteContext=@n1.ExecuteContext)
   SequenceAggregate_1 = &RigVMFunction_Sequence(ExecuteContext=@SequenceAggregate.B)
   RigVMFunctionReturnNode(A=@SequenceAggregate.A, B=@SequenceAggregate_1.A, C=@SequenceAggregate_1.B)
  ]}
]
hierarchy [
 Start {type=Connector, loc=<0 0 0>, rot=<0 0 0 1>}
 Root {type=Socket, loc=<0 0 0>, rot=<0 0 0 1>}
 End {type=Connector, loc=<0 0 0>, rot=<0 0 0 1>}
 EndParent {type=Connector, loc=<0 0 0>, rot=<0 0 0 1>}
 StartParent {type=Connector, loc=<0 0 0>, rot=<0 0 0 1>}
]
variables [
 Alt Color {type=FLinearColor, default=(R=0,G=0,B=0,A=1)}
 Bones {type="TArray<FRigElementKey>"}
 Chest Null {type=FRigElementKey, default=(Type=Bone,Name="")}
 Color {type=FLinearColor, default=(R=1,G=1,B=1,A=1)}
 Control Names As FK {type=bool, default=True}
 Control Scale {type=double, default=1}
 Control Scale Factor Profile {
  type=FRuntimeFloatCurve
  default=(EditorCurveData=(Keys=((Value=1),(Time=1,Value=1)),DefaultValue=340282346638528859811704183484516925440,PreInfinityExtrap=RCCE_Constant,PostInfinityExtrap=RCCE_Constant),ExternalCurve=None)
 }
 Debug {type=bool, default=False}
 Default Distribute Rotation {type=double, default=0}
 End Control {type=FRigElementKey, default=(Type=Bone,Name="")}
 End Driver {type=FRigElementKey, default=(Type=Bone,Name="")}
 End FK Space Nulls {type="TArray<FRigElementKey>"}
 End Fk Spaces {type="TArray<FRigElementKey>"}
 End IK Space Nulls {type="TArray<FRigElementKey>"}
 End Moveable Pivot Display Name {type=FName, default=None}
 End Parent {type=FRigElementKey, default=(Type=None,Name="")}
 End Position {type=FVector, default=(X=0,Y=0,Z=0)}
 End Proxy Buffer {type=FRigElementKey, default=(Type=None,Name="")}
 End Proxy Control {type=FRigElementKey, default=(Type=None,Name="")}
 End Proxy Vis Channel {type=FRigElementKey, default=(Type=Bone,Name="")}
 End Spaces {type="TArray<FRigElementKey>"}
 FK Bones {type="TArray<FRigElementKey>"}
 FK Controls {type="TArray<FRigElementKey>"}
 FK Gimbal Controls {type="TArray<FRigElementKey>"}
 FK Parent {type=FRigElementKey, default=(Type=Bone,Name="")}
 FK Primary Display Names {type="TArray<FName>"}
 FK Secondary Display Names {type="TArray<FName>"}
 FKs Control Shapes {
  type=FRigUnit_HierarchyAddControl_ShapeSettings
  default=(bVisible=True,Name="Circle_Thin",Color=(R=1,G=1,B=1,A=1),Transform=(Rotation=(X=0,Y=0,Z=0,W=1),Translation=(X=0,Y=0,Z=0),Scale3D=(X=3.7,Y=3.7,Z=2)))
 }
 FKs Visibility Channel {type=FRigElementKey, default=(Type=Bone,Name="")}
 IK End Display Name {type=FName, default=None}
 IK Gimbal Controls {type="TArray<FRigElementKey>"}
 IK Mid Display Name {type=FName, default=None}
 IK Start Display Name {type=FName, default=None}
 IKs Control Shapes {
  type=FRigUnit_HierarchyAddControl_ShapeSettings
  default=(bVisible=True,Name="Square_Thin",Color=(R=1,G=1,B=1,A=1),Transform=(Rotation=(X=0,Y=0,Z=0,W=1),Translation=(X=0,Y=0,Z=0),Scale3D=(X=3.5,Y=3.5,Z=3)))
 }
 IKs Visibility Channel {type=FRigElementKey, default=(Type=Bone,Name="")}
 Is Neck {type=bool, default=False}
 Match Null Parent {type=FRigElementKey, default=(Type=Bone,Name="")}
 Match Nulls {type="TArray<FRigElementKey>", default=(())}
 Mid Control {type=FRigElementKey, default=(Type=Bone,Name="")}
 Mid Null {type=FRigElementKey, default=(Type=Bone,Name="")}
 Mid Smart Match Channel {type=FRigElementKey, default=(Type=Bone,Name="")}
 Orient Null {type=FRigElementKey, default=(Type=Bone,Name="")}
 Proxy Slide Channel {type=FRigElementKey, default=(Type=Bone,Name="")}
 Rotation Offset {type=FQuat, default=(X=0,Y=0,Z=0,W=1)}
 Rotation Order {type=EEulerRotationOrder, default=XYZ}
 Sec FK Controls {type="TArray<FRigElementKey>"}
 Sec FK Parent {type=FRigElementKey, default=(Type=Bone,Name="")}
 "Sec FKs Control Shapes " {
  type=FRigUnit_HierarchyAddControl_ShapeSettings
  default=(bVisible=True,Name="Triangle_Thin",Color=(R=1,G=1,B=1,A=1),Transform=(Rotation=(X=0,Y=-0.7071,Z=0,W=0.7071),Translation=(X=0,Y=-13,Z=0),Scale3D=(X=1,Y=1,Z=3)))
 }
 Sec FKs Visibility Channel {type=FRigElementKey, default=(Type=Bone,Name="")}
 Spline {type=FControlRigSpline, default=()}
 Spline Orig Length {type=double, default=0}
 Start Control {type=FRigElementKey, default=(Type=Bone,Name="")}
 Start Driver {type=FRigElementKey, default=(Type=Bone,Name="")}
 Start FK Space Nulls {type="TArray<FRigElementKey>"}
 Start Fk Spaces {type="TArray<FRigElementKey>"}
 Start IK Space Nulls {type="TArray<FRigElementKey>"}
 Start Parent {type=FRigElementKey, default=(Type=None,Name="")}
 Start Position {type=FVector, default=(X=0,Y=0,Z=0)}
 Start Spaces {type="TArray<FRigElementKey>"}
 Stretch Channel {type=FRigElementKey, default=(Type=Bone,Name="")}
 Up Vector Axis {type=FVector, default=(X=0,Y=0,Z=0)}
]
